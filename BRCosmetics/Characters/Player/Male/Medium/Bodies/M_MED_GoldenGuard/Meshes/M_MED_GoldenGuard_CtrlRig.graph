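# Control Rig for the GoldenGuard male body that poses its dynamic skirt, belt, scarf and front-flap bones to follow the body across locomotion states.
# Crouch, front-end, movement, jump, skydive and slide variables, smoothed by AlphaInterp, weight additive bone offsets and thigh or calf constraints.
# Spherical pose readers on the thighs drive corrective skirt offsets.
# Belt bones are parent-constrained to a weighted blend of front skirt and pelvis bones, then offset.
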ControlRigBlueprint /BRCosmetics/Characters/Player/Male/Medium/Bodies/M_MED_GoldenGuard/Meshes/M_MED_GoldenGuard_CtrlRig
graphs [
 ContainedGraph {nodes [
   n1 = RigVMFunctionEntryNode()
   n2 = &RigVMFunction_MathDoubleMul(A=@n1.A, B=@n1.B)
   RigVMFunctionReturnNode(Result=&RigVMFunction_MathDoubleMul(B=@n1.D, A=&RigVMFunction_MathDoubleMul(B=@n1.C, A=@n2.Result).Result).Result)
  ]}
 ContainedGraph {nodes [
   n1 = RigVMFunctionEntryNode()
   RigVMFunctionReturnNode(Result=&RigVMFunction_MathDoubleMul(B=@n1.D, A=&RigVMFunction_MathDoubleMul(A=@n1.A, B=@n1.C).Result).Result)
  ]}
 RigVMFunctionLibrary {links [], nodes []}
 RigVMModel {nodes [
   RigUnit_BeginExecution = &RigUnit_BeginExecution()
   n1 = /Script/ControlRig.RigUnit_SequenceExecution(ExecuteContext=@RigUnit_BeginExecution.ExecutePin) {is=struct RigVMUnitNode}
   n2 = &RigUnit_SequenceExecution(ExecuteContext=@n1.C)
   Branch_RigVMUnitNode = &RigVMFunction_ControlFlowBranch(ExecuteContext=@n2.A, Condition=RigVMVariableNode() {variable=Is_Crouching}.Value)
   n3 = &RigUnit_SequenceExecution(ExecuteContext=@Branch_RigVMUnitNode.False)
   n4 = &RigUnit_SphericalPoseReader(
    ActiveRegionScaleFactors=(PositiveWidth=4,NegativeWidth=1,PositiveHeight=4,NegativeHeight=4)
    ActiveRegionSize=0.04
    Debug=(bDrawDebug=false,bDraw2D=false,DebugScale=25,DebugSegments=20,DebugThickness=0.25)
    DriverAxis=(X=-1,Y=0,Z=0)
    DriverItem=(Type=Bone,Name="thigh_l")
    FalloffRegionScaleFactors=(PositiveWidth=0.5,NegativeWidth=0.4,PositiveHeight=1,NegativeHeight=1)
    FalloffSize=0.5
    OptionalParentItem=(Type=Bone)
    RotationOffset=(X=160,Y=-50,Z=0)
    ExecutePin=@n3.A
   )
   > &RigUnit_ModifyTransforms(
    ItemToModify=((Item=(Type=Bone,Name="dyn_skirt_l_r"),Transform=(Rotation=(X=0,Y=0,Z=0,W=1),Translation=(X=0,Y=-5,Z=0),Scale3D=(X=1,Y=1,Z=1))),(Item=(Type=Bone,Name="dyn_skirt_l"),Transform=(Rotation=(X=0,Y=0,Z=0,W=1),Translation=(X=0,Y=-2,Z=0),Scale3D=(X=1,Y=1,Z=1))))
    ItemToModify.0.Transform.Rotation=&RigVMFunction_MathQuaternionFromEuler(Euler=(X=-20,Y=0,Z=-10)).Result
    Weight=@n4.OutputParam
   )
   n5 = &RigUnit_SphericalPoseReader(
    ActiveRegionScaleFactors=(PositiveWidth=1,NegativeWidth=1,PositiveHeight=1,NegativeHeight=1)
    ActiveRegionSize=0.05
    Debug=(bDrawDebug=false,bDraw2D=false,DebugScale=25,DebugSegments=20,DebugThickness=0.25)
    DriverAxis=(X=1,Y=0,Z=0)
    DriverItem=(Type=Bone,Name="thigh_r")
    FalloffRegionScaleFactors=(PositiveWidth=0.5,NegativeWidth=1,PositiveHeight=1,NegativeHeight=1)
    FalloffSize=0.6
    OptionalParentItem=(Type=Bone)
    RotationOffset=(X=-80,Y=0,Z=40)
    ExecutePin=@n3.B
   )
   > &RigUnit_ModifyTransforms(
    ItemToModify=((Item=(Type=Bone,Name="dyn_skirt_r_l"),Transform=(Rotation=(X=0,Y=0,Z=0,W=1),Translation=(X=0,Y=-5,Z=0),Scale3D=(X=1,Y=1,Z=1))),(Item=(Type=Bone,Name="dyn_skirt_r"),Transform=(Rotation=(X=0,Y=0,Z=0,W=1),Translation=(X=0,Y=-5,Z=0),Scale3D=(X=1,Y=1,Z=1))))
    ItemToModify.0.Transform.Rotation=&RigVMFunction_MathQuaternionFromEuler(Euler=(X=0,Y=0,Z=-10)).Result
    Weight=@n5.OutputParam
   )
   n6 = &RigVMFunction_AlphaInterp(
    InRange=(Max=1)
    InterpSpeedDecreasing=5
    InterpSpeedIncreasing=10
    OutRange=(Max=1)
    Scale=1
    ClampMin=~
    bMapRange=~
    Value=RigVMVariableNode() {variable=is_sprinting_alpha}.Value
   )
   &RigUnit_ModifyTransforms(
    ItemToModify="((Item=(Type=Bone,Name=\"dyn_skirt_b_b\"),Transform=(Rotation=(X=0,Y=0,Z=0,W=1),Translation=(X=0,Y=-2,Z=0),Scale3D=(X=1,Y=1,Z=1))),(Item=(Type=Bone,Name=\"dyn_skirt_l_l_1\"),Transform=(Rotation=(X=0,Y=0,Z=0,W=1),Translation=(X=0,Y=-5,Z=0),Scale3D=(X=1,Y=1,Z=1) ...(truncated)"
    Weight=@n6.Result
    ItemToModify.0.Transform.Rotation=&RigVMFunction_MathQuaternionFromEuler(Euler=(X=0,Y=10,Z=0)).Result
    ExecutePin=@n3.C
   )
   n7 = &RigVMFunction_AlphaInterp(
    InRange=(Min=0.1,Max=1)
    InterpSpeedDecreasing=2.5
    InterpSpeedIncreasing=5
    OutRange=(Max=1)
    Scale=1
    bClampResult=true
    bMapRange=true
    ClampMin=~
    Value=&RigVMFunction_MathDoubleSub(A=RigVMVariableNode() {variable=is_moving_2d_alpha}.Value, B=RigVMVariableNode() {variable=is_skydiving}.Value).Result
   )
   &RigUnit_ModifyTransforms(
    ItemToModify="((Item=(Type=Bone,Name=\"dyn_f_main\"),Transform=(Rotation=(X=0,Y=0,Z=0,W=1),Translation=(X=2,Y=-1,Z=0),Scale3D=(X=1,Y=1,Z=1))),(Item=(Type=Bone,Name=\"dyn_skirt_f_f\"),Transform=(Rotation=(X=0,Y=0,Z=0,W=1),Translation=(X=0,Y=-3,Z=0),Scale3D=(X=1,Y=1,Z=0.5))), ...(truncated)"
    Weight=@n7.Result
    ExecutePin=@n3.D
   )
   n8 = RigVMVariableNode() {variable=is_moving_2d_alpha}
   n9 = RigVMVariableNode() {variable=is_jumping_alpha}
   n10 = &RigVMFunction_AlphaInterp(
    InRange=(Min=0.1,Max=1)
    InterpSpeedDecreasing=2.5
    InterpSpeedIncreasing=5
    OutRange=(Max=1)
    Scale=1
    bClampResult=true
    bMapRange=true
    ClampMin=~
    Value=&RigVMFunction_MathDoubleSub(A=&RigVMFunction_MathDoubleSub(B=@n8.Value, A=@n9.Value).Result, B=RigVMVariableNode() {variable=is_skydiving}.Value).Result
   )
   > &RigUnit_ModifyTransforms(
    ItemToModify="((Item=(Type=Bone,Name=\"dyn_f_main\"),Transform=(Rotation=(X=0,Y=0,Z=0,W=1),Translation=(X=2,Y=-1,Z=0),Scale3D=(X=1,Y=1,Z=1))),(Item=(Type=Bone,Name=\"dyn_skirt_f_f\"),Transform=(Rotation=(X=0,Y=0,Z=0,W=1),Translation=(X=0,Y=-3,Z=0),Scale3D=(X=1,Y=1,Z=0.5))), ...(truncated)"
    Weight=@n10.Result
   )
   n11 = &RigUnit_SphericalPoseReader(
    ActiveRegionScaleFactors=(PositiveWidth=1,NegativeWidth=1,PositiveHeight=3,NegativeHeight=3)
    ActiveRegionSize=0.2
    Debug=(bDrawDebug=false,bDraw2D=false,DebugScale=25,DebugSegments=20,DebugThickness=0.25)
    DriverAxis=(X=1,Y=0,Z=0)
    DriverItem=(Type=Bone,Name="thigh_r")
    FalloffRegionScaleFactors=(PositiveWidth=0.5,NegativeWidth=1,PositiveHeight=1,NegativeHeight=1)
    FalloffSize=0.5
    FlipHeightScaling=false
    FlipWidthScaling=false
    OptionalParentItem=(Type=Bone)
    RotationOffset=(X=-90,Y=0,Z=0)
    ExecutePin=@Branch_RigVMUnitNode.True
   )
   AlphaInterp = &RigVMFunction_AlphaInterp(InterpSpeedDecreasing=5, InterpSpeedIncreasing=10, Scale=1, bClampResult=False, Value=@n11.OutputParam)
   > &RigUnit_ModifyTransforms(
    ItemToModify="((Item=(Type=Bone,Name=\"dyn_f_main\"),Transform=(Rotation=(X=0,Y=0,Z=0,W=1),Translation=(X=-2,Y=-2,Z=0),Scale3D=(X=1,Y=1,Z=1))),(Item=(Type=Bone,Name=\"dyn_skirt_f_f\"),Transform=(Rotation=(X=0,Y=0,Z=0,W=1),Translation=(X=0,Y=-5,Z=0),Scale3D=(X=1,Y=1,Z=0.5))) ...(truncated)"
    ItemToModify.0.Transform.Rotation=&RigVMFunction_MathQuaternionFromEuler(Euler=(X=0,Y=0,Z=-30)).Result
    ItemToModify.1.Transform.Rotation=&RigVMFunction_MathQuaternionFromEuler(Euler=(X=0,Y=0,Z=35)).Result
    Weight=@AlphaInterp.Result
   )
   > &RigUnit_ModifyTransforms(
    ItemToModify="((Item=(Type=Bone,Name=\"dyn_belt_sub_l\"),Transform=(Rotation=(X=0,Y=0,Z=0,W=1),Translation=(X=0,Y=0,Z=2),Scale3D=(X=1,Y=1,Z=1))),(Item=(Type=Bone,Name=\"dyn_belt_f\"),Transform=(Rotation=(X=0,Y=0,Z=0,W=1),Translation=(X=0,Y=0,Z=-2),Scale3D=(X=1,Y=1,Z=1))),(I ...(truncated)"
    ItemToModify.2.Transform.Rotation=&RigVMFunction_MathQuaternionFromEuler(Euler=(X=0,Y=0,Z=0)).Result
    ItemToModify.1.Transform.Rotation=&RigVMFunction_MathQuaternionFromEuler(Euler=(X=0,Y=0,Z=0)).Result
    ItemToModify.0.Transform.Rotation=&RigVMFunction_MathQuaternionFromEuler(Euler=(X=0,Y=-30,Z=0)).Result
    Weight=@AlphaInterp.Result
   )
   n12 = &RigVMFunction_AlphaInterp(
    InRange=(Min=0.1,Max=1)
    InterpSpeedDecreasing=2.5
    InterpSpeedIncreasing=5
    OutRange=(Max=0.5)
    Scale=1
    bClampResult=true
    bMapRange=true
    ClampMin=~
    Value=RigVMVariableNode() {variable=left_right_alpha}.Value
   )
   > &RigUnit_RotationConstraintLocalSpaceOffset(
    AdvancedSettings=(RotationOrderForFilter=XZY)
    Child=(Type=Bone,Name="dyn_skirt_f_f")
    Filter=(bX=True,bY=True,bZ=True)
    Parents=((Item=(Type=Bone,Name="thigh_l"),Weight=1))
    Weight=@n12.Result
   )
   n13 = &RigVMFunction_AlphaInterp(
    InRange=(Min=-0.1,Max=-1)
    InterpSpeedDecreasing=2.5
    InterpSpeedIncreasing=5
    OutRange=(Max=0.5)
    Scale=1
    bClampResult=true
    bMapRange=true
    ClampMin=~
    Value=RigVMVariableNode() {variable=left_right_alpha}.Value
   )
   > &RigUnit_RotationConstraintLocalSpaceOffset(
    AdvancedSettings=(RotationOrderForFilter=XZY)
    Child=(Type=Bone,Name="dyn_skirt_f_f")
    Filter=(bX=True,bY=True,bZ=True)
    Parents=((Item=(Type=Bone,Name="thigh_r"),Weight=1))
    Weight=@n13.Result
   )
   Branch_1_RigVMUnitNode = &RigVMFunction_ControlFlowBranch(ExecuteContext=@n2.B, Condition=RigVMVariableNode() {variable=is_in_front_end}.Value)
   n14 = &RigUnit_SequenceExecution(ExecuteContext=@Branch_1_RigVMUnitNode.False)
   n15 = &RigUnit_SphericalPoseReader(
    ActiveRegionScaleFactors=(PositiveWidth=1,NegativeWidth=1,PositiveHeight=3,NegativeHeight=3)
    ActiveRegionSize=0.2
    Debug=(bDrawDebug=false,bDraw2D=false,DebugScale=25,DebugSegments=20,DebugThickness=0.25)
    DriverAxis=(X=1,Y=0,Z=0)
    DriverItem=(Type=Bone,Name="thigh_r")
    FalloffRegionScaleFactors=(PositiveWidth=0.5,NegativeWidth=1,PositiveHeight=1,NegativeHeight=1)
    FalloffSize=0.5
    FlipHeightScaling=false
    FlipWidthScaling=false
    OptionalParentItem=(Type=Bone)
    RotationOffset=(X=-90,Y=0,Z=0)
    ExecutePin=@n14.B
   )
   Multiply_1_1 = RigVMAggregateNode(
    D=@n15.OutputParam
    C=&RigVMFunction_MathDoubleSub(A=1, B=RigVMVariableNode() {variable=is_rigidbody_active_alpha}.Value).Result
    A=RigVMVariableNode() {variable=Is_Crouching_Alpha}.Value
   )
   n16 = &RigVMFunction_AlphaInterp(InterpSpeedDecreasing=10, InterpSpeedIncreasing=10, Scale=1, bClampResult=true, bMapRange=true, Value=@Multiply_1_1.Result)
   > &RigUnit_ModifyTransforms(
    ItemToModify=((Item=(Type=Bone,Name="dyn_f_main"),Transform=(Rotation=(X=0,Y=0,Z=0,W=1),Translation=(X=5,Y=15,Z=0),Scale3D=(X=1,Y=1,Z=1))))
    Weight=@n16.Result
    ItemToModify.0.Transform.Rotation=&RigVMFunction_MathQuaternionFromEuler(Euler=(X=0,Y=0,Z=0)).Result
   )
   n17 = &RigVMFunction_AlphaInterp(
    InterpSpeedDecreasing=100
    InterpSpeedIncreasing=5
    OutRange=(Max=1)
    Scale=1
    bClampResult=true
    bMapRange=true
    ClampMin=~
    Value=&RigVMFunction_MathDoubleSub(A=1, B=RigVMVariableNode() {variable=Is_Crouching_Alpha}.Value).Result
   )
   Multiply_1 = RigVMAggregateNode(
    A=&RigVMFunction_MathDoubleSub(A=1, B=RigVMVariableNode() {variable=is_moving_2d_alpha}.Value).Result
    B=&RigVMFunction_MathDoubleSub(A=1, B=RigVMVariableNode() {variable=is_jumping_alpha}.Value).Result
    C=&RigVMFunction_MathDoubleSub(A=1, B=RigVMVariableNode() {variable=is_skydiving}.Value).Result
    D=&RigVMFunction_MathDoubleSub(A=1, B=RigVMVariableNode() {variable=fall}.Value).Result
   )
   n18 = &RigVMFunction_AlphaInterp(
    InterpSpeedDecreasing=5
    InterpSpeedIncreasing=6
    OutRange=(Max=1)
    Scale=1
    bClampResult=true
    bMapRange=true
    ClampMin=~
    OutRange.Max=@n17.Result
    Value=@Multiply_1.Result
   )
   &RigUnit_ParentConstraint(
    AdvancedSettings=(RotationOrderForFilter=XZY)
    Child=(Type=Bone,Name="dyn_skirt_l_1")
    Parents=((Item=(Type=Bone,Name="thigh_l"),Weight=1))
    Weight=@n18.Result
    ExecutePin=@n14.A
   )
   > &RigUnit_ParentConstraint(
    AdvancedSettings=(RotationOrderForFilter=XZY)
    Child=(Type=Bone,Name="dyn_skirt_l_l_1")
    Filter=(TranslationFilter=(bX=false,bY=false,bZ=false),RotationFilter=(bX=True,bY=True,bZ=True),ScaleFilter=(bX=True,bY=True,bZ=True))
    Parents=((Item=(Type=Bone,Name="thigh_l"),Weight=1),(Item=(Type=Bone,Name="calf_l"),Weight=0.4))
    Weight=@n18.Result
   )
   > &RigUnit_ParentConstraint(
    AdvancedSettings=(RotationOrderForFilter=XZY)
    Child=(Type=Bone,Name="dyn_skirt_l_r_1")
    Parents=((Item=(Type=Bone,Name="thigh_l"),Weight=1))
    Weight=@n18.Result
   )
   > &RigUnit_ParentConstraint(
    AdvancedSettings=(RotationOrderForFilter=XZY)
    Child=(Type=Bone,Name="dyn_skirt_r_1")
    Parents=((Item=(Type=Bone,Name="thigh_r"),Weight=1))
    Weight=@n18.Result
   )
   > &RigUnit_ParentConstraint(
    AdvancedSettings=(RotationOrderForFilter=XZY)
    Child=(Type=Bone,Name="dyn_skirt_r_l_1")
    Parents=((Item=(Type=Bone,Name="thigh_r"),Weight=1))
    Weight=@n18.Result
   )
   > &RigUnit_ParentConstraint(
    AdvancedSettings=(RotationOrderForFilter=XZY)
    Child=(Type=Bone,Name="dyn_skirt_r_r_1")
    Filter=(TranslationFilter=(bX=false,bY=false,bZ=false),RotationFilter=(bX=True,bY=True,bZ=True),ScaleFilter=(bX=True,bY=True,bZ=True))
    Parents=((Item=(Type=Bone,Name="thigh_r"),Weight=1),(Item=(Type=Bone,Name="calf_r"),Weight=0.4))
    Weight=@n18.Result
   )
   > &RigUnit_ParentConstraint(
    AdvancedSettings=(RotationOrderForFilter=XZY)
    Child=(Type=Bone,Name="dyn_skirt_l_2")
    Filter=(TranslationFilter=(bX=false,bY=false,bZ=false),RotationFilter=(bX=True,bY=True,bZ=True),ScaleFilter=(bX=True,bY=True,bZ=True))
    Parents=((Item=(Type=Bone,Name="thigh_l"),Weight=1),(Item=(Type=Bone,Name="calf_l"),Weight=0.8))
    Weight=@n18.Result
   )
   > &RigUnit_ParentConstraint(
    AdvancedSettings=(RotationOrderForFilter=XZY)
    Child=(Type=Bone,Name="dyn_skirt_l_l_2")
    Filter=(TranslationFilter=(bX=false,bY=false,bZ=false),RotationFilter=(bX=True,bY=True,bZ=True),ScaleFilter=(bX=True,bY=True,bZ=True))
    Parents=((Item=(Type=Bone,Name="thigh_l"),Weight=0),(Item=(Type=Bone,Name="calf_l"),Weight=0.8))
    Weight=@n18.Result
   )
   > &RigUnit_ParentConstraint(
    AdvancedSettings=(RotationOrderForFilter=XZY)
    Child=(Type=Bone,Name="dyn_skirt_l_r_2")
    Parents=((Item=(Type=Bone,Name="thigh_l"),Weight=1))
    Weight=@n18.Result
   )
   > &RigUnit_ParentConstraint(
    AdvancedSettings=(RotationOrderForFilter=XZY)
    Child=(Type=Bone,Name="dyn_skirt_r_2")
    Filter=(TranslationFilter=(bX=false,bY=false,bZ=false),RotationFilter=(bX=True,bY=True,bZ=True),ScaleFilter=(bX=True,bY=True,bZ=True))
    Parents=((Item=(Type=Bone,Name="thigh_r"),Weight=1),(Item=(Type=Bone,Name="calf_r"),Weight=0.7))
    Weight=@n18.Result
   )
   > &RigUnit_ParentConstraint(
    AdvancedSettings=(RotationOrderForFilter=XZY)
    Child=(Type=Bone,Name="dyn_skirt_r_l_2")
    Parents=((Item=(Type=Bone,Name="thigh_r"),Weight=1))
    Weight=@n18.Result
   )
   > &RigUnit_ParentConstraint(
    AdvancedSettings=(RotationOrderForFilter=XZY)
    Child=(Type=Bone,Name="dyn_skirt_r_r_2")
    Filter=(TranslationFilter=(bX=false,bY=false,bZ=false),RotationFilter=(bX=True,bY=True,bZ=True),ScaleFilter=(bX=True,bY=True,bZ=True))
    Parents=((Item=(Type=Bone,Name="thigh_r"),Weight=0),(Item=(Type=Bone,Name="calf_r"),Weight=0.7))
    Weight=@n18.Result
   )
   n19 = &RigVMFunction_MathQuaternionFromEuler(Euler=(X=0,Y=0,Z=-4))
   n20 = &RigVMFunction_AlphaInterp(
    InRange=(Max=1)
    InterpSpeedDecreasing=5
    InterpSpeedIncreasing=2
    OutRange=(Max=1)
    Scale=1
    bClampResult=true
    bMapRange=true
    ClampMin=~
    Value=&RigVMFunction_MathDoubleSub(A=1, B=RigVMVariableNode() {variable=is_moving_2d_alpha}.Value).Result
   )
   > &RigUnit_ModifyTransforms(
    ItemToModify="((Item=(Type=Bone,Name=\"dyn_skirt_b_b\"),Transform=(Rotation=(X=0,Y=0,Z=0,W=1),Translation=(X=0,Y=-2,Z=0),Scale3D=(X=1,Y=1,Z=1))),(Item=(Type=Bone,Name=\"dyn_skirt_l\"),Transform=(Rotation=(X=0,Y=0,Z=0,W=1),Translation=(X=0,Y=-1.4,Z=0),Scale3D=(X=1,Y=1,Z=1))) ...(truncated)"
    ItemToModify.0.Transform.Rotation=&RigVMFunction_MathQuaternionFromEuler(Euler=(X=0,Y=0,Z=0)).Result
    ItemToModify.1.Transform.Rotation=@n19.Result
    ItemToModify.3.Transform.Rotation=@n19.Result
    ItemToModify.4.Transform.Rotation=@n19.Result
    ItemToModify.5.Transform.Rotation=@n19.Result
    ItemToModify.6.Transform.Rotation=@n19.Result
    ItemToModify.2.Transform.Rotation=&RigVMFunction_MathQuaternionFromEuler(Euler=(X=0,Y=0,Z=0)).Result
    Weight=@n20.Result
   )
   n21 = &RigVMFunction_MathQuaternionFromEuler(Euler=(X=0,Y=0,Z=8))
   > &RigUnit_ModifyTransforms(
    ItemToModify="((Item=(Type=Bone,Name=\"dyn_skirt_b_b_2\"),Transform=(Rotation=(X=0,Y=0,Z=0,W=1),Translation=(X=0,Y=0,Z=0),Scale3D=(X=1,Y=1,Z=1))),(Item=(Type=Bone,Name=\"dyn_skirt_l_2\"),Transform=(Rotation=(X=0,Y=0,Z=0,W=1),Translation=(X=0,Y=2,Z=0),Scale3D=(X=1,Y=1,Z=1))) ...(truncated)"
    ItemToModify.0.Transform.Rotation=&RigVMFunction_MathQuaternionFromEuler(Euler=(X=0,Y=0,Z=0)).Result
    ItemToModify.1.Transform.Rotation=@n21.Result
    ItemToModify.2.Transform.Rotation=@n21.Result
    ItemToModify.3.Transform.Rotation=@n21.Result
    ItemToModify.4.Transform.Rotation=@n21.Result
    ItemToModify.5.Transform.Rotation=@n21.Result
    ItemToModify.6.Transform.Rotation=@n21.Result
    Weight=@n20.Result
   )
   n22 = &RigVMFunction_MathQuaternionFromEuler(Euler=(X=0,Y=0,Z=8))
   > &RigUnit_ModifyTransforms(
    ItemToModify="((Item=(Type=Bone,Name=\"dyn_skirt_l_r_3\"),Transform=(Rotation=(X=0,Y=0,Z=0,W=1),Translation=(X=0,Y=0,Z=0),Scale3D=(X=1,Y=1,Z=1))),(Item=(Type=Bone,Name=\"dyn_skirt_l_3\"),Transform=(Rotation=(X=0,Y=0,Z=0,W=1),Translation=(X=0,Y=1,Z=0),Scale3D=(X=1,Y=1,Z=1))) ...(truncated)"
    Weight=1
    ItemToModify.0.Transform.Rotation=@n22.Result
    ItemToModify.1.Transform.Rotation=@n22.Result
   )
   > &RigUnit_ModifyTransforms(
    ItemToModify="((Item=(Type=Bone,Name=\"dyn_scarf_f_1\"),Transform=(Rotation=(X=0,Y=0,Z=0,W=1),Translation=(X=0,Y=0,Z=0),Scale3D=(X=1,Y=1,Z=1))),(Item=(Type=Bone,Name=\"dyn_skirt_f_f\"),Transform=(Rotation=(X=0,Y=0,Z=0,W=1),Translation=(X=0,Y=1,Z=0),Scale3D=(X=1,Y=1,Z=1))),( ...(truncated)"
    Weight=1
    ItemToModify.3.Transform.Rotation=&RigVMFunction_MathQuaternionFromEuler(Euler=(X=0,Y=-6,Z=0)).Result
    ItemToModify.0.Transform.Rotation=&RigVMFunction_MathQuaternionFromEuler(Euler=(X=0,Y=-6,Z=-10)).Result
    ItemToModify.1.Transform.Rotation=&RigVMFunction_MathQuaternionFromEuler(Euler=(X=0,Y=0,Z=0)).Result
   )
   n23 = &RigVMFunction_AlphaInterp(
    InterpSpeedDecreasing=10
    InterpSpeedIncreasing=2
    OutRange=(Max=1)
    Scale=1
    bClampResult=true
    bMapRange=true
    ClampMin=~
    Value=RigVMVariableNode() {variable=is_moving_2d_alpha}.Value
   )
   n24 = &RigVMFunction_MathQuaternionFromEuler(Euler=(X=0,Y=0,Z=-6))
   > &RigUnit_ModifyTransforms(
    ItemToModify="((Item=(Type=Bone,Name=\"dyn_skirt_b_b\"),Transform=(Rotation=(X=0,Y=0,Z=0,W=1),Translation=(X=0,Y=-4,Z=0),Scale3D=(X=1,Y=1,Z=1))),(Item=(Type=Bone,Name=\"dyn_skirt_l\"),Transform=(Rotation=(X=0,Y=0,Z=0,W=1),Translation=(X=0,Y=-1.4,Z=0),Scale3D=(X=1,Y=1,Z=1))) ...(truncated)"
    Weight=@n23.Result
    ItemToModify.0.Transform.Rotation=&RigVMFunction_MathQuaternionFromEuler(Euler=(X=0,Y=-4,Z=0)).Result
    ItemToModify.1.Transform.Rotation=@n24.Result
    ItemToModify.3.Transform.Rotation=@n24.Result
    ItemToModify.4.Transform.Rotation=@n24.Result
    ItemToModify.5.Transform.Rotation=@n24.Result
    ItemToModify.6.Transform.Rotation=@n24.Result
    ItemToModify.7.Transform.Rotation=&RigVMFunction_MathQuaternionFromEuler(Euler=(X=0,Y=0,Z=0)).Result
    ItemToModify.2.Transform.Rotation=&RigVMFunction_MathQuaternionFromEuler(Euler=(X=10,Y=0,Z=-6)).Result
   )
   n25 = &RigVMFunction_AlphaInterp(
    InRange=(Min=0,Max=-1)
    InterpSpeedDecreasing=10
    InterpSpeedIncreasing=10
    OutRange=(Max=1)
    Scale=1
    bClampResult=true
    bMapRange=true
    ClampMin=~
    Value=&RigVMFunction_MathDoubleMul(B=RigVMVariableNode() {variable=left_right_alpha}.Value, A=RigVMVariableNode() {variable=Is_Crouching_Alpha}.Value).Result
   )
   n26 = &RigVMFunction_MathQuaternionFromEuler(Euler=(X=0,Y=0,Z=-20))
   > &RigUnit_ModifyTransforms(
    ItemToModify=((Item=(Type=Bone,Name="dyn_skirt_l_r"),Transform=(Rotation=(X=0,Y=0,Z=0,W=1),Translation=(X=0,Y=0,Z=0),Scale3D=(X=1,Y=1,Z=1))),(Item=(Type=Bone,Name="dyn_skirt_l"),Transform=(Rotation=(X=0,Y=0,Z=0,W=1),Translation=(X=0,Y=1,Z=0),Scale3D=(X=1,Y=1,Z=1))))
    Weight=@n25.Result
    ItemToModify.0.Transform.Rotation=@n26.Result
    ItemToModify.1.Transform.Rotation=@n26.Result
   )
   n27 = &RigVMFunction_MathQuaternionFromEuler(Euler=(X=0,Y=0,Z=3))
   n28 = &RigVMFunction_AlphaInterp(
    InRange=(Max=1)
    InterpSpeedDecreasing=5
    InterpSpeedIncreasing=2
    OutRange=(Max=1)
    Scale=1
    bClampResult=true
    bMapRange=true
    ClampMin=~
    Value=&RigVMFunction_MathDoubleSub(A=1, B=RigVMVariableNode() {variable=is_moving_2d_alpha}.Value).Result
   )
   &RigUnit_ModifyTransforms(
    ItemToModify="((Item=(Type=Bone,Name=\"dyn_skirt_b_b\"),Transform=(Rotation=(X=0,Y=0,Z=0,W=1),Translation=(X=0,Y=-2,Z=0),Scale3D=(X=1,Y=1,Z=1))),(Item=(Type=Bone,Name=\"dyn_skirt_l\"),Transform=(Rotation=(X=0,Y=0,Z=0,W=1),Translation=(X=0,Y=-1.4,Z=0),Scale3D=(X=1,Y=1,Z=1))) ...(truncated)"
    ExecutePin=@Branch_1_RigVMUnitNode.True
    ItemToModify.1.Transform.Rotation=@n27.Result
    ItemToModify.2.Transform.Rotation=@n27.Result
    ItemToModify.3.Transform.Rotation=@n27.Result
    ItemToModify.4.Transform.Rotation=@n27.Result
    ItemToModify.5.Transform.Rotation=@n27.Result
    ItemToModify.6.Transform.Rotation=@n27.Result
    ItemToModify.0.Transform.Rotation=&RigVMFunction_MathQuaternionFromEuler(Euler=(X=0,Y=0,Z=0)).Result
    Weight=@n28.Result
   )
   > &RigUnit_ModifyTransforms(
    ItemToModify="((Item=(Type=Bone,Name=\"dyn_skirt_b_b_2\"),Transform=(Rotation=(X=0,Y=0,Z=0,W=1),Translation=(X=0,Y=1,Z=0),Scale3D=(X=1,Y=1,Z=1))),(Item=(Type=Bone,Name=\"dyn_skirt_l_2\"),Transform=(Rotation=(X=0,Y=0,Z=0,W=1),Translation=(X=0,Y=2,Z=0),Scale3D=(X=1,Y=1,Z=1))) ...(truncated)"
    ItemToModify.0.Transform.Rotation=&RigVMFunction_MathQuaternionFromEuler(Euler=(X=0,Y=0,Z=5)).Result
    Weight=@n28.Result
   )
   > &RigUnit_ModifyTransforms(
    ItemToModify="((Item=(Type=Bone,Name=\"dyn_scarf_f_1\"),Transform=(Rotation=(X=0,Y=0,Z=0,W=1),Translation=(X=0,Y=0,Z=0),Scale3D=(X=1,Y=1,Z=1))),(Item=(Type=Bone,Name=\"dyn_skirt_f_f\"),Transform=(Rotation=(X=0,Y=0,Z=0,W=1),Translation=(X=0,Y=1,Z=0),Scale3D=(X=1,Y=1,Z=1))),( ...(truncated)"
    Weight=1
    ItemToModify.1.Transform.Rotation=&RigVMFunction_MathQuaternionFromEuler(Euler=(X=0,Y=0,Z=0)).Result
    ItemToModify.0.Transform.Rotation=&RigVMFunction_MathQuaternionFromEuler(Euler=(X=0,Y=-6,Z=-10)).Result
    ItemToModify.3.Transform.Rotation=&RigVMFunction_MathQuaternionFromEuler(Euler=(X=0,Y=-6,Z=0)).Result
   )
   n29 = &RigUnit_SequenceExecution(ExecuteContext=@n1.D)
   n30 = &RigVMFunction_AlphaInterp(
    InRange=(Max=1)
    InterpSpeedDecreasing=5
    InterpSpeedIncreasing=10
    OutRange=(Max=1)
    Scale=1
    ClampMin=~
    bMapRange=~
    Value=RigVMVariableNode() {variable=is_knee_sliding_alpha}.Value
   )
   &RigUnit_ModifyTransforms(
    ItemToModify=((Item=(Type=Bone,Name="dyn_skirt_b_b"),Transform=(Rotation=(X=0,Y=0,Z=0,W=1),Translation=(X=-7,Y=0,Z=0),Scale3D=(X=1,Y=1,Z=1))))
    ItemToModify.0.Transform.Rotation=&RigVMFunction_MathQuaternionFromEuler(Euler=(X=0,Y=0,Z=-90)).Result
    ExecutePin=@n29.A
    Weight=@n30.Result
   )
   > &RigUnit_ModifyTransforms(
    ItemToModify=((Item=(Type=Bone,Name="dyn_skirt_l_r"),Transform=(Rotation=(X=0,Y=0,Z=0,W=1),Translation=(X=0,Y=0,Z=0),Scale3D=(X=1,Y=1,Z=1))))
    ItemToModify.0.Transform.Rotation=&RigVMFunction_MathQuaternionFromEuler(Euler=(X=-30,Y=0,Z=0)).Result
    Weight=@n30.Result
   )
   > &RigUnit_ModifyTransforms(
    ItemToModify=((Item=(Type=Bone,Name="dyn_skirt_l"),Transform=(Rotation=(X=0,Y=0,Z=0,W=1),Translation=(X=0,Y=0,Z=0),Scale3D=(X=1,Y=1,Z=1))))
    ItemToModify.0.Transform.Rotation=&RigVMFunction_MathQuaternionFromEuler(Euler=(X=0,Y=0,Z=-20)).Result
    Weight=@n30.Result
   )
   > &RigUnit_ModifyTransforms(
    ItemToModify=((Item=(Type=Bone,Name="dyn_skirt_l_l"),Transform=(Rotation=(X=0,Y=0,Z=0,W=1),Translation=(X=-5,Y=2,Z=5),Scale3D=(X=1,Y=1,Z=1))))
    ItemToModify.0.Transform.Rotation=&RigVMFunction_MathQuaternionFromEuler(Euler=(X=-30,Y=0,Z=-60)).Result
    Weight=@n30.Result
   )
   > &RigUnit_ModifyTransforms(
    ItemToModify=((Item=(Type=Bone,Name="dyn_skirt_r_l"),Transform=(Rotation=(X=0,Y=0,Z=0,W=1),Translation=(X=0,Y=0,Z=0),Scale3D=(X=1,Y=1,Z=1))))
    ItemToModify.0.Transform.Rotation=&RigVMFunction_MathQuaternionFromEuler(Euler=(X=0,Y=0,Z=0)).Result
    Weight=@n30.Result
   )
   > &RigUnit_ModifyTransforms(
    ItemToModify=((Item=(Type=Bone,Name="dyn_skirt_r_r"),Transform=(Rotation=(X=0,Y=0,Z=0,W=1),Translation=(X=-6,Y=0,Z=0),Scale3D=(X=1,Y=1,Z=1))))
    ItemToModify.0.Transform.Rotation=&RigVMFunction_MathQuaternionFromEuler(Euler=(X=0,Y=0,Z=-70)).Result
    Weight=@n30.Result
   )
   > &RigUnit_ModifyTransforms(
    ItemToModify=((Item=(Type=Bone,Name="dyn_skirt_r"),Transform=(Rotation=(X=0,Y=0,Z=0,W=1),Translation=(X=0,Y=0,Z=0),Scale3D=(X=1,Y=1,Z=1))))
    ItemToModify.0.Transform.Rotation=&RigVMFunction_MathQuaternionFromEuler(Euler=(X=0,Y=0,Z=-10)).Result
    Weight=@n30.Result
   )
   n31 = &RigVMFunction_AlphaInterp(
    InRange=(Max=1)
    InterpSpeedDecreasing=5
    InterpSpeedIncreasing=10
    OutRange=(Max=1)
    Scale=1
    ClampMin=~
    bMapRange=~
    Value=RigVMVariableNode() {variable=is_baseball_sliding_alpha}.Value
   )
   &RigUnit_ModifyTransforms(
    ItemToModify=((Item=(Type=Bone,Name="dyn_skirt_b_b"),Transform=(Rotation=(X=0,Y=0,Z=0,W=1),Translation=(X=-7,Y=0,Z=0),Scale3D=(X=1,Y=1,Z=1))))
    ItemToModify.0.Transform.Rotation=&RigVMFunction_MathQuaternionFromEuler(Euler=(X=0,Y=0,Z=-90)).Result
    ExecutePin=@n29.B
    Weight=@n31.Result
   )
   > &RigUnit_ModifyTransforms(
    ItemToModify=((Item=(Type=Bone,Name="dyn_skirt_l_r"),Transform=(Rotation=(X=0,Y=0,Z=0,W=1),Translation=(X=0,Y=0,Z=0),Scale3D=(X=1,Y=1,Z=1))))
    ItemToModify.0.Transform.Rotation=&RigVMFunction_MathQuaternionFromEuler(Euler=(X=0,Y=0,Z=-5)).Result
    Weight=@n31.Result
   )
   > &RigUnit_ModifyTransforms(
    ItemToModify=((Item=(Type=Bone,Name="dyn_f_main"),Transform=(Rotation=(X=0,Y=0,Z=0,W=1),Translation=(X=0,Y=3,Z=0),Scale3D=(X=1,Y=1,Z=1))),(Item=(Type=Bone,Name="dyn_skirt_f_f"),Transform=(Rotation=(X=0,Y=0,Z=0,W=1),Translation=(X=0,Y=2,Z=0),Scale3D=(X=1,Y=1,Z=1))))
    ItemToModify.0.Transform.Rotation=&RigVMFunction_MathQuaternionFromEuler(Euler=(X=0,Y=0,Z=60)).Result
    ItemToModify.1.Transform.Rotation=&RigVMFunction_MathQuaternionFromEuler(Euler=(X=0,Y=0,Z=-30)).Result
    Weight=@n31.Result
   )
   > &RigUnit_ModifyTransforms(
    ItemToModify=((Item=(Type=Bone,Name="dyn_skirt_l_l"),Transform=(Rotation=(X=0,Y=0,Z=0,W=1),Translation=(X=0,Y=0,Z=0),Scale3D=(X=1,Y=1,Z=1))))
    ItemToModify.0.Transform.Rotation=&RigVMFunction_MathQuaternionFromEuler(Euler=(X=0,Y=0,Z=-30)).Result
    Weight=@n31.Result
   )
   > &RigUnit_ModifyTransforms(
    ItemToModify=((Item=(Type=Bone,Name="dyn_skirt_r_l"),Transform=(Rotation=(X=0,Y=0,Z=0,W=1),Translation=(X=-4,Y=0,Z=0),Scale3D=(X=1,Y=1,Z=1))))
    ItemToModify.0.Transform.Rotation=&RigVMFunction_MathQuaternionFromEuler(Euler=(X=0,Y=0,Z=-90)).Result
    Weight=@n31.Result
   )
   > &RigUnit_ModifyTransforms(
    ItemToModify=((Item=(Type=Bone,Name="dyn_skirt_r_r"),Transform=(Rotation=(X=0,Y=0,Z=0,W=1),Translation=(X=-6,Y=0,Z=0),Scale3D=(X=1,Y=1,Z=1))))
    ItemToModify.0.Transform.Rotation=&RigVMFunction_MathQuaternionFromEuler(Euler=(X=0,Y=0,Z=-90)).Result
    Weight=@n31.Result
   )
   > &RigUnit_ModifyTransforms(
    ItemToModify=((Item=(Type=Bone,Name="dyn_skirt_r"),Transform=(Rotation=(X=0,Y=0,Z=0,W=1),Translation=(X=-5,Y=0,Z=0),Scale3D=(X=1,Y=1,Z=1))))
    ItemToModify.0.Transform.Rotation=&RigVMFunction_MathQuaternionFromEuler(Euler=(X=0,Y=0,Z=-90)).Result
    Weight=@n31.Result
   )
   > &RigUnit_ModifyTransforms(
    ItemToModify=((Item=(Type=Bone,Name="dyn_skirt_l"),Transform=(Rotation=(X=0,Y=0,Z=0,W=1),Translation=(X=0,Y=0,Z=0),Scale3D=(X=1,Y=1,Z=1))))
    ItemToModify.0.Transform.Rotation=&RigVMFunction_MathQuaternionFromEuler(Euler=(X=0,Y=0,Z=-20)).Result
    Weight=@n31.Result
   )
   &RigUnit_ParentConstraint(
    AdvancedSettings=(RotationOrderForFilter=XZY)
    Child=(Type=Bone,Name="dyn_belt")
    Parents=((Item=(Type=Bone,Name="pelvis"),Weight=0.0771),(Item=(Type=Bone,Name="dyn_skirt_f_f"),Weight=0.4257),(Item=(Type=Bone,Name="dyn_skirt_f_f_1"),Weight=0.2533),(Item=(Type=Bone,Name="dyn_f_main"),Weight=0.2439))
    Weight=1
    ExecutePin=@n29.C
   )
   > &RigUnit_ParentConstraint(
    AdvancedSettings=(RotationOrderForFilter=XZY)
    Child=(Type=Bone,Name="dyn_belt_1")
    Parents=((Item=(Type=Bone,Name="dyn_skirt_f_f_2"),Weight=0.0863),(Item=(Type=Bone,Name="dyn_skirt_f_f"),Weight=0.4475),(Item=(Type=Bone,Name="dyn_skirt_f_f_1"),Weight=0.4662))
    Weight=1
   )
   > &RigUnit_ParentConstraint(
    AdvancedSettings=(RotationOrderForFilter=XZY)
    Child=(Type=Bone,Name="dyn_belt_2")
    Parents=((Item=(Type=Bone,Name="dyn_skirt_f_f_3"),Weight=0.0224),(Item=(Type=Bone,Name="dyn_skirt_f_f"),Weight=0.2398),(Item=(Type=Bone,Name="dyn_skirt_f_f_1"),Weight=0.5415),(Item=(Type=Bone,Name="dyn_skirt_f_f_2"),Weight=0.1964))
    Weight=1
   )
   > &RigUnit_ModifyTransforms(
    ItemToModify="((Item=(Type=Bone,Name=\"dyn_belt\"),Transform=(Rotation=(X=0,Y=0,Z=0,W=1),Translation=(X=0,Y=-1,Z=0),Scale3D=(X=1,Y=1,Z=1))),(Item=(Type=Bone,Name=\"dyn_belt_1\"),Transform=(Rotation=(X=0,Y=0,Z=0,W=1),Translation=(X=0,Y=0,Z=0),Scale3D=(X=1,Y=1,Z=1))),(Item=(T ...(truncated)"
    Weight=1
    ItemToModify.0.Transform.Rotation=&RigVMFunction_MathQuaternionFromEuler(Euler=(X=0,Y=0,Z=0)).Result
    ItemToModify.1.Transform.Rotation=&RigVMFunction_MathQuaternionFromEuler(Euler=(X=0,Y=0,Z=0)).Result
    ItemToModify.2.Transform.Rotation=&RigVMFunction_MathQuaternionFromEuler(Euler=(X=0,Y=0,Z=0)).Result
    ItemToModify.4.Transform.Rotation=&RigVMFunction_MathQuaternionFromEuler(Euler=(X=-6,Y=0,Z=3)).Result
    ItemToModify.5.Transform.Rotation=&RigVMFunction_MathQuaternionFromEuler(Euler=(X=0,Y=0,Z=10)).Result
    ItemToModify.3.Transform.Rotation=&RigVMFunction_MathQuaternionFromEuler(Euler=(X=0,Y=-20,Z=0)).Result
   )
   Multiply = &RigVMFunction_MathDoubleMul(B=RigVMVariableNode() {variable=Is_Crouching_Alpha}.Value, A=RigVMVariableNode() {variable=is_moving_2d_alpha}.Value)
   n32 = &RigVMFunction_AlphaInterp(
    InRange=(Max=1)
    InterpSpeedDecreasing=5
    InterpSpeedIncreasing=10
    OutRange=(Max=1)
    Scale=1
    bClampResult=true
    bMapRange=true
    ClampMin=~
    Value=@Multiply.Result
   )
   > &RigUnit_ModifyTransforms(
    ItemToModify="((Item=(Type=Bone,Name=\"dyn_belt_sub_l\"),Transform=(Rotation=(X=0,Y=0,Z=0,W=1),Translation=(X=-1,Y=-5,Z=4),Scale3D=(X=1,Y=1,Z=1))),(Item=(Type=Bone,Name=\"dyn_belt_sub_r\"),Transform=(Rotation=(X=0,Y=0,Z=0,W=1),Translation=(X=3,Y=0,Z=-5),Scale3D=(X=1,Y=1,Z=1 ...(truncated)"
    Weight=@n32.Result
    ItemToModify.0.Transform.Rotation=&RigVMFunction_MathQuaternionFromEuler(Euler=(X=0,Y=-40,Z=-10)).Result
    ItemToModify.1.Transform.Rotation=&RigVMFunction_MathQuaternionFromEuler(Euler=(X=-5,Y=100,Z=-30)).Result
    ItemToModify.2.Transform.Rotation=&RigVMFunction_MathQuaternionFromEuler(Euler=(X=0,Y=0,Z=-10)).Result
    ItemToModify.3.Transform.Rotation=&RigVMFunction_MathQuaternionFromEuler(Euler=(X=0,Y=-20,Z=0)).Result
   )
   n33 = &RigVMFunction_AlphaInterp(
    InRange=(Max=1)
    InterpSpeedDecreasing=5
    InterpSpeedIncreasing=10
    OutRange=(Max=1)
    Scale=1
    bClampResult=true
    bMapRange=true
    ClampMin=~
    Value=RigVMVariableNode() {variable=is_moving_2d_alpha}.Value
   )
   > &RigUnit_ModifyTransforms(
    ItemToModify=((Item=(Type=Bone,Name="dyn_skirt_l"),Transform=(Rotation=(X=0,Y=0,Z=0,W=1),Translation=(X=0,Y=0,Z=0),Scale3D=(X=1,Y=1,Z=1))))
    Weight=@n33.Result
    ItemToModify.0.Transform.Rotation=&RigVMFunction_MathQuaternionFromEuler(Euler=(X=0,Y=0,Z=2)).Result
   )
   n34 = &RigVMFunction_AlphaInterp(
    InRange=(Max=1)
    InterpSpeedDecreasing=5
    InterpSpeedIncreasing=10
    OutRange=(Max=1)
    Scale=1
    bClampResult=true
    bMapRange=true
    ClampMin=~
    Value=RigVMVariableNode() {variable=is_skydiving}.Value
   )
   n35 = &RigVMFunction_MathQuaternionFromEuler(Euler=(X=0,Y=20,Z=-10))
   > &RigUnit_ModifyTransforms(
    ItemToModify=((Item=(Type=Bone,Name="dyn_belt_sub_r"),Transform=(Rotation=(X=0,Y=0,Z=0,W=1),Translation=(X=2,Y=0,Z=0),Scale3D=(X=1,Y=1,Z=1))),(Item=(Type=Bone,Name="dyn_belt_f"),Transform=(Rotation=(X=0,Y=0,Z=0,W=1),Translation=(X=0,Y=0,Z=0),Scale3D=(X=1,Y=1,Z=1))))
    Weight=@n34.Result
    ItemToModify.0.Transform.Rotation=@n35.Result
    ItemToModify.1.Transform.Rotation=@n35.Result
   )
   n36 = &RigVMFunction_AlphaInterp(
    InRange=(Max=1)
    InterpSpeedDecreasing=5
    InterpSpeedIncreasing=10
    OutRange=(Max=1)
    Scale=1
    ClampMin=~
    bMapRange=~
    Value=RigVMVariableNode() {variable=Is_Crouching_Alpha}.Value
   )
   &RigUnit_ModifyTransforms(
    ItemToModify=((Item=(Type=Bone,Name="dyn_skirt_b_b"),Transform=(Rotation=(X=0,Y=0,Z=0,W=1),Translation=(X=0,Y=-4,Z=0),Scale3D=(X=1,Y=1,Z=1))))
    ItemToModify.0.Transform.Rotation=&RigVMFunction_MathQuaternionFromEuler(Euler=(X=0,Y=10,Z=0)).Result
    ExecutePin=@n1.A
    Weight=@n36.Result
   )
   > &RigUnit_ModifyTransforms(
    ItemToModify=((Item=(Type=Bone,Name="dyn_skirt_l_r"),Transform=(Rotation=(X=0,Y=0,Z=0,W=1),Translation=(X=0,Y=-5,Z=0),Scale3D=(X=1,Y=1,Z=1))))
    ItemToModify.0.Transform.Rotation=&RigVMFunction_MathQuaternionFromEuler(Euler=(X=0,Y=0,Z=0)).Result
    Weight=@n36.Result
   )
   > &RigUnit_ModifyTransforms(
    ItemToModify=((Item=(Type=Bone,Name="dyn_skirt_r_l"),Transform=(Rotation=(X=0,Y=0,Z=0,W=1),Translation=(X=0,Y=-2,Z=0),Scale3D=(X=1,Y=1,Z=1))))
    ItemToModify.0.Transform.Rotation=&RigVMFunction_MathQuaternionFromEuler(Euler=(X=0,Y=0,Z=-10)).Result
    Weight=@n36.Result
   )
   > &RigUnit_ModifyTransforms(
    ItemToModify=((Item=(Type=Bone,Name="dyn_skirt_r_r"),Transform=(Rotation=(X=0,Y=0,Z=0,W=1),Translation=(X=0,Y=-1,Z=0),Scale3D=(X=1,Y=1,Z=1))))
    ItemToModify.0.Transform.Rotation=&RigVMFunction_MathQuaternionFromEuler(Euler=(X=0,Y=0,Z=-10)).Result
    Weight=@n36.Result
   )
   > &RigUnit_ModifyTransforms(
    ItemToModify=((Item=(Type=Bone,Name="dyn_skirt_r"),Transform=(Rotation=(X=0,Y=0,Z=0,W=1),Translation=(X=0,Y=-1,Z=0),Scale3D=(X=1,Y=1,Z=1))))
    ItemToModify.0.Transform.Rotation=&RigVMFunction_MathQuaternionFromEuler(Euler=(X=0,Y=0,Z=-10)).Result
    Weight=@n36.Result
   )
   > &RigUnit_ModifyTransforms(
    ItemToModify=((Item=(Type=Bone,Name="dyn_scarf_f"),Transform=(Rotation=(X=0,Y=0,Z=0,W=1),Translation=(X=0,Y=-1,Z=0),Scale3D=(X=1,Y=1,Z=1))))
    ItemToModify.0.Transform.Rotation=&RigVMFunction_MathQuaternionFromEuler(Euler=(X=0,Y=-10,Z=0)).Result
    Weight=@n36.Result
   )
   n37 = &RigUnit_SphericalPoseReader(
    ActiveRegionScaleFactors=(PositiveWidth=1,NegativeWidth=1,PositiveHeight=3,NegativeHeight=3)
    ActiveRegionSize=0.2
    Debug=(bDrawDebug=false,bDraw2D=false,DebugScale=25,DebugSegments=20,DebugThickness=0.25)
    DriverAxis=(X=1,Y=0,Z=0)
    DriverItem=(Type=Bone,Name="thigh_r")
    FalloffRegionScaleFactors=(PositiveWidth=0.5,NegativeWidth=1,PositiveHeight=1,NegativeHeight=1)
    FalloffSize=0.5
    FlipHeightScaling=false
    FlipWidthScaling=false
    OptionalParentItem=(Type=Bone)
    RotationOffset=(X=-90,Y=0,Z=0)
    ExecutePin=@n1.B
   )
   Subtract = &RigVMFunction_MathDoubleSub(A=RigVMVariableNode() {variable=Is_Crouching_Alpha}.Value, B=RigVMVariableNode() {variable=is_moving_2d_alpha}.Value)
   n38 = &RigVMFunction_AlphaInterp(
    InRange=(Max=1)
    InterpSpeedDecreasing=5
    InterpSpeedIncreasing=5
    OutRange=(Max=1)
    Scale=1
    bClampResult=true
    bMapRange=true
    ClampMin=~
    Value=&RigVMFunction_MathDoubleSub(B=@n37.OutputParam, A=@Subtract.Result).Result
   )
   > &RigUnit_ModifyTransforms(
    ItemToModify="((Item=(Type=Bone,Name=\"dyn_f_main\"),Transform=(Rotation=(X=0,Y=0,Z=0,W=1),Translation=(X=-1,Y=0,Z=0),Scale3D=(X=1,Y=1,Z=1))),(Item=(Type=Bone,Name=\"dyn_skirt_f_f\"),Transform=(Rotation=(X=0,Y=0,Z=0,W=1),Translation=(X=0,Y=0,Z=0),Scale3D=(X=1,Y=1,Z=0.5))),( ...(truncated)"
    Weight=@n38.Result
    ItemToModify.0.Transform.Rotation=&RigVMFunction_MathQuaternionFromEuler(Euler=(X=14,Y=0,Z=25)).Result
    ItemToModify.3.Transform.Rotation=&RigVMFunction_MathQuaternionFromEuler(Euler=(X=0,Y=10,Z=-30)).Result
    ItemToModify.4.Transform.Rotation=&RigVMFunction_MathQuaternionFromEuler(Euler=(X=0,Y=10,Z=-30)).Result
    ItemToModify.2.Transform.Rotation=&RigVMFunction_MathQuaternionFromEuler(Euler=(X=0,Y=10,Z=-30)).Result
    ItemToModify.1.Transform.Rotation=&RigVMFunction_MathQuaternionFromEuler(Euler=(X=40,Y=10,Z=25)).Result
   )
   > &RigUnit_ModifyTransforms(
    ItemToModify="((Item=(Type=Bone,Name=\"dyn_belt_sub_l\"),Transform=(Rotation=(X=0,Y=0,Z=0,W=1),Translation=(X=0,Y=0,Z=2),Scale3D=(X=1,Y=1,Z=1))),(Item=(Type=Bone,Name=\"dyn_skirt_l_r\"),Transform=(Rotation=(X=0,Y=0,Z=0,W=1),Translation=(X=0,Y=0,Z=-2),Scale3D=(X=1,Y=1,Z=1))) ...(truncated)"
    Weight=1
    ItemToModify.2.Transform.Rotation=&RigVMFunction_MathQuaternionFromEuler(Euler=(X=0,Y=0,Z=-10)).Result
    ItemToModify.1.Transform.Rotation=&RigVMFunction_MathQuaternionFromEuler(Euler=(X=0,Y=0,Z=0)).Result
    ItemToModify.0.Transform.Rotation=&RigVMFunction_MathQuaternionFromEuler(Euler=(X=0,Y=-30,Z=0)).Result
   )
   &RigVMFunction_MathQuaternionFromEuler(Euler=(X=0,Y=0,Z=0))
   RigVMRerouteNode(Value=())
   &RigVMFunction_MathDoubleSub(A=1, B=@n15.OutputParam)
   &RigVMFunction_MathDoubleSub(A=1, B=RigVMVariableNode() {variable=is_moving_2d_alpha}.Value)
  ]}
]
hierarchy [
 root {loc=<0 0 0>, rot=<0 0 0 1>}
 pelvis {parent=root, loc=<0 -0.7071 93.1191>, rot=<0.0127 0.707 -0.0127 -0.707>}
 spine_01 {parent=pelvis, loc=<6.7494 -0.6388 0>, rot=<0 0 0.034 -0.9994>}
 spine_02 {parent=spine_01, loc=<7.2518 -0.035 0>, rot=<0 0 0.0238 -0.9997>}
 spine_03 {parent=spine_02, loc=<6.0104 0 0>, rot=<0 0 0.0149 -0.9999>}
 spine_04 {parent=spine_03, loc=<8.1024 0 0>, rot=<0 0 -0.0132 -0.9999>}
 spine_05 {parent=spine_04, loc=<10.0146 0 0>, rot=<0 0 -0.0015 -1>}
 clavicle_l {parent=spine_05, loc=<4.3575 -0.1466 -2.3223>, rot=<0.0465 -0.7289 -0.009 -0.6829>}
 upperarm_l {parent=clavicle_l, loc=<15.5111 0 0>, rot=<0.0226 -0.3234 -0.0088 -0.9459>}
 lowerarm_l {parent=upperarm_l, loc=<26.6514 0 0>, rot=<0.0391 -0.0099 -0.2453 -0.9686>}
 hand_l {parent=lowerarm_l, loc=<22.3491 0 0>, rot=<-0.7081 -0.0507 0.0975 -0.6975>}
 index_metacarpal_l {parent=hand_l, loc=<1.6535 0.0183 -2.96>, rot=<0.0158 -0.0124 -0.0173 -0.9996>}
 index_01_l {parent=index_metacarpal_l, loc=<8.3008 0.0357 -0.0524>, rot=<-0.0044 0.0485 0.0924 -0.9945>}
 index_02_l {parent=index_01_l, loc=<3.3587 -0.3004 0.0452>, rot=<-0.002 0.0027 0.3225 -0.9466>}
 index_03_l {parent=index_02_l, loc=<2.5578 -0.123 -0.0252>, rot=<-0.002 0.0067 0.2558 -0.9667>}
 middle_metacarpal_l {parent=hand_l, loc=<2.1856 -0.1725 -0.5984>, rot=<0.006 0.0168 -0.0329 -0.9993>}
 middle_01_l {parent=middle_metacarpal_l, loc=<7.3275 -0.2447 0.0129>, rot=<-0.0208 0.0042 0.0858 -0.9961>}
 middle_02_l {parent=middle_01_l, loc=<4.0123 0.1401 0.1175>, rot=<-0.0218 0.0258 0.3512 -0.9357>}
 middle_03_l {parent=middle_02_l, loc=<2.6077 0.0534 0.1373>, rot=<0.0082 0.0188 0.1708 -0.9851>}
 pinky_metacarpal_l {parent=hand_l, loc=<1.7616 -0.9753 3.3814>, rot=<-0.0397 0.0814 -0.069 -0.9935>}
 pinky_01_l {parent=pinky_metacarpal_l, loc=<7.1709 -0.4831 -0.1094>, rot=<-0.0654 -0.0035 0.2242 -0.9723>}
 pinky_02_l {parent=pinky_01_l, loc=<2.8274 0.0893 -0.1043>, rot=<0.0037 0.0011 0.2854 -0.9584>}
 pinky_03_l {parent=pinky_02_l, loc=<2.2046 0.0307 -0.0078>, rot=<-0.0002 0.0098 0.1445 -0.9895>}
 ring_metacarpal_l {parent=hand_l, loc=<2.2074 -0.3625 1.3638>, rot=<-0.0284 0.0722 -0.0458 -0.9959>}
 ring_01_l {parent=ring_metacarpal_l, loc=<7.0295 -0.6197 -0.1446>, rot=<-0.0489 -0.0252 0.1892 -0.9804>}
 ring_02_l {parent=ring_01_l, loc=<3.6174 0.4647 -0.1027>, rot=<0 0 0.2806 -0.9598>}
 ring_03_l {parent=ring_02_l, loc=<2.4152 0.114 0.0199>, rot=<0 0 0.1189 -0.9929>}
 thumb_01_l {parent=hand_l, loc=<1.3118 -0.3191 -2.459>, rot=<0.6546 -0.4042 0.0664 -0.6354>}
 thumb_02_l {parent=thumb_01_l, loc=<5.9336 -0.0448 0.0401>, rot=<-0.0037 0.0213 0.298 -0.9543>}
 thumb_03_l {parent=thumb_02_l, loc=<2.2223 0.3682 -0.0907>, rot=<0.0118 -0.013 0.0066 -0.9998>}
 weapon_l {parent=hand_l, loc=<-0.1974 -3.4111 0.3811>, rot=<0 0 0 1>}
 lowerarm_twist_01_l {parent=lowerarm_l, loc=<17.8549 0 0>, rot=<0 0 0 1>}
 lowerarm_twist_02_l {parent=lowerarm_l, loc=<9.6199 0 0>, rot=<0 0 0 1>}
 upperarm_twist_01_l {parent=upperarm_l, loc=<0.0041 0 0>, rot=<0 0 0 1>}
 upperarm_twist_02_l {parent=upperarm_l, loc=<13.7941 0 0>, rot=<0 0 0 1>}
 dyn_shoulder_pad_l {parent=upperarm_l, loc=<5.2959 -0.1733 6.1044>, rot=<-0.0174 0.6568 -0.0085 -0.7538>}
 clavicle_r {parent=spine_05, loc=<4.3575 -0.1466 2.3223>, rot=<0.7289 0.0465 -0.6829 0.009>}
 upperarm_r {parent=clavicle_r, loc=<-15.5111 0 0>, rot=<0.0226 -0.3234 -0.0088 -0.9459>}
 lowerarm_r {parent=upperarm_r, loc=<-26.6515 0 0>, rot=<0.0391 -0.0099 -0.2453 -0.9686>}
 hand_r {parent=lowerarm_r, loc=<-22.3492 0 0>, rot=<-0.7081 -0.0507 0.0975 -0.6975>}
 index_metacarpal_r {parent=hand_r, loc=<-1.6534 -0.0181 2.9601>, rot=<0.0158 -0.0124 -0.0173 -0.9996>}
 index_01_r {parent=index_metacarpal_r, loc=<-8.3008 -0.0357 0.0524>, rot=<-0.0044 0.0485 0.0924 -0.9945>}
 index_02_r {parent=index_01_r, loc=<-3.3589 0.3003 -0.0453>, rot=<-0.002 0.0027 0.3225 -0.9466>}
 index_03_r {parent=index_02_r, loc=<-2.5578 0.1231 0.0251>, rot=<-0.002 0.0067 0.2558 -0.9667>}
 middle_metacarpal_r {parent=hand_r, loc=<-2.1856 0.1726 0.5984>, rot=<0.006 0.0168 -0.0329 -0.9993>}
 middle_01_r {parent=middle_metacarpal_r, loc=<-7.3273 0.2447 -0.0129>, rot=<-0.0208 0.0042 0.0858 -0.9961>}
 middle_02_r {parent=middle_01_r, loc=<-4.0142 -0.1411 -0.07>, rot=<-0.0217 0.0259 0.3512 -0.9357>}
 middle_03_r {parent=middle_02_r, loc=<-2.6074 -0.0534 -0.1312>, rot=<0.0083 0.0188 0.1708 -0.9851>}
 pinky_metacarpal_r {parent=hand_r, loc=<-1.7617 0.9753 -3.3814>, rot=<-0.0397 0.0814 -0.069 -0.9935>}
 pinky_01_r {parent=pinky_metacarpal_r, loc=<-7.1707 0.4832 0.1094>, rot=<-0.0654 -0.0035 0.2242 -0.9723>}
 pinky_02_r {parent=pinky_01_r, loc=<-2.8274 -0.0892 0.1043>, rot=<0.0037 0.0011 0.2854 -0.9584>}
 pinky_03_r {parent=pinky_02_r, loc=<-2.2045 -0.031 0.0078>, rot=<-0.0002 0.0098 0.1445 -0.9895>}
 ring_metacarpal_r {parent=hand_r, loc=<-2.2073 0.3627 -1.3638>, rot=<-0.0284 0.0722 -0.0458 -0.9959>}
 ring_01_r {parent=ring_metacarpal_r, loc=<-7.0297 0.6197 0.1446>, rot=<-0.0489 -0.0252 0.1892 -0.9804>}
 ring_02_r {parent=ring_01_r, loc=<-3.6171 -0.4648 0.1027>, rot=<0 0 0.2806 -0.9598>}
 ring_03_r {parent=ring_02_r, loc=<-2.4153 -0.1141 -0.0199>, rot=<0 0 0.1189 -0.9929>}
 thumb_01_r {parent=hand_r, loc=<-1.3119 0.3192 2.459>, rot=<0.6546 -0.4042 0.0664 -0.6354>}
 thumb_02_r {parent=thumb_01_r, loc=<-5.9335 0.0447 -0.0401>, rot=<-0.0037 0.0213 0.298 -0.9543>}
 thumb_03_r {parent=thumb_02_r, loc=<-2.2223 -0.3682 0.0907>, rot=<0.0118 -0.013 0.0066 -0.9998>}
 weapon_r {parent=hand_r, loc=<0.1975 3.4112 -0.3811>, rot=<0 0 0 -1>}
 lowerarm_twist_01_r {parent=lowerarm_r, loc=<-17.855 0 0>, rot=<0 0 0 1>}
 lowerarm_twist_02_r {parent=lowerarm_r, loc=<-9.6199 0 0>, rot=<0 0 0 1>}
 upperarm_twist_01_r {parent=upperarm_r, loc=<-0.0041 0 0>, rot=<0 0 0 1>}
 upperarm_twist_02_r {parent=upperarm_r, loc=<-13.7942 0 0>, rot=<0 0 0 1>}
 dyn_shoulder_pad_r {parent=upperarm_r, loc=<-5.2962 0.1733 -6.104>, rot=<-0.0174 0.6568 -0.0085 -0.7538>}
 neck_01 {parent=spine_05, loc=<11.3138 0.1529 0>, rot=<0 0 -0.2003 -0.9797>}
 neck_02 {parent=neck_01, loc=<4.5627 0 0>, rot=<0 0 0.0504 -0.9987>}
 head {parent=neck_02, loc=<5.3914 0 0>, rot=<0 0 0.115 -0.9934>}
 dyn_scarf_main {parent=spine_05, loc=<11.6543 0.2595 0>, rot=<0 0 -0.1512 -0.9885>}
 dyn_scarf_f {parent=spine_05, loc=<5.5418 11.4774 3.7649>, rot=<0.1216 0.6978 0.6982 0.104>}
 dyn_scarf_f_1 {parent=dyn_scarf_f, loc=<14.0539 0 0>, rot=<-0.0027 -0.1463 0.0005 -0.9892>}
 dyn_scarf_f_2 {parent=dyn_scarf_f_1, loc=<14.6013 0 0>, rot=<-0.0003 -0.0595 0.0002 -0.9982>}
 dyn_scarf_f_3 {parent=dyn_scarf_f_2, loc=<11.1948 0 0>, rot=<0 0 0 1>}
 dyn_scarf_b {parent=spine_05, loc=<12.7682 -8.4137 -6.4896>, rot=<0.2407 -0.9701 0.005 -0.0302>}
 dyn_scarf_b_1 {parent=dyn_scarf_b, loc=<10.1583 0 0>, rot=<0 0 -0.1646 -0.9864>}
 dyn_scarf_b_2 {parent=dyn_scarf_b_1, loc=<10.9468 0 0>, rot=<0 0 0 1>}
 thigh_l {parent=pelvis, loc=<-2.5693 -0.0109 -7.7501>, rot=<0.0012 0.0572 0.0155 -0.9982>}
 calf_l {parent=thigh_l, loc=<-39.0484 0 0>, rot=<0.0006 -0.0161 -0.0391 -0.9991>}
 calf_twist_01_l {parent=calf_l, loc=<-35.813 -0.0967 0.0397>, rot=<0.0237 0.0095 -0.0035 -0.9997>}
 calf_twist_02_l {parent=calf_l, loc=<-18.8064 -0.0967 0.0397>, rot=<0.0237 0.0095 -0.0035 -0.9997>}
 foot_l {parent=calf_l, loc=<-42.561 0.1077 0.4292>, rot=<0.0418 -0.0455 0.0343 -0.9975>}
 ball_l {parent=foot_l, loc=<-6.1599 13.8264 0.6806>, rot=<0.0001 -0.0001 -0.7063 -0.7079>}
 thigh_twist_01_l {parent=thigh_l, loc=<-21.6303 0 0>, rot=<0.0013 0 0 -1>}
 thigh_r {parent=pelvis, loc=<-2.5691 -0.0109 7.7501>, rot=<0.0572 -0.0012 0.9982 0.0155>}
 calf_r {parent=thigh_r, loc=<39.0487 0 0>, rot=<0.0006 -0.0161 -0.0391 -0.9991>}
 calf_twist_01_r {parent=calf_r, loc=<35.8132 0.0967 -0.0397>, rot=<0.0237 0.0095 -0.0035 -0.9997>}
 calf_twist_02_r {parent=calf_r, loc=<18.8064 0.0967 -0.0397>, rot=<0.0237 0.0095 -0.0035 -0.9997>}
 foot_r {parent=calf_r, loc=<42.5612 -0.1077 -0.4292>, rot=<-0.0031 -0.0411 0.0414 -0.9983>}
 ball_r {parent=foot_r, loc=<6.1598 -13.8264 -0.6806>, rot=<0.0001 -0.0001 -0.7063 -0.7079>}
 thigh_twist_01_r {parent=thigh_r, loc=<21.6303 0 0>, rot=<0.0013 0 0 -1>}
 dyn_bag {parent=pelvis, loc=<4.4876 4.7753 -12.4395>, rot=<-0.1995 -0.6784 -0.1751 -0.6851>}
 dyn_f_main {parent=pelvis, loc=<7.5831 8.7404 -0.0005>, rot=<0.0001 0 0.9979 0.0655>}
 dyn_skirt_f_f {parent=dyn_f_main, loc=<7.3142 -0.8742 -0.0001>, rot=<0 -0.0002 -0.0389 -0.9992>}
 dyn_skirt_f_f_1 {parent=dyn_skirt_f_f, loc=<10.5174 0 -0.0001>, rot=<0 0.0001 0 -1>}
 dyn_skirt_f_f_2 {parent=dyn_skirt_f_f_1, loc=<16.3784 0 0>, rot=<0 0 0 1>}
 dyn_skirt_f_f_3 {parent=dyn_skirt_f_f_2, loc=<11.8946 0 0>, rot=<0 0 0 1>}
 dyn_belt_f {parent=dyn_f_main, loc=<5.1443 -3.3028 -3.7975>, rot=<0 -0.0001 -0.0476 -0.9989>}
 dyn_belt {parent=dyn_belt_f, loc=<4.1607 0 0>, rot=<0 0 0 -1>}
 dyn_belt_1 {parent=dyn_belt, loc=<8.3497 0 0>, rot=<0 0 0 -1>}
 dyn_belt_2 {parent=dyn_belt_1, loc=<8.0015 0 0>, rot=<0 0 0 1>}
 dyn_skirt_l_r {parent=pelvis, loc=<1.9848 1.8664 -13.7044>, rot=<0.0589 -0.731 -0.6752 -0.0788>}
 dyn_skirt_l_r_1 {parent=dyn_skirt_l_r, loc=<11.8156 0 0>, rot=<0 0 0 1>}
 dyn_skirt_l_r_2 {parent=dyn_skirt_l_r_1, loc=<13.3633 0 0>, rot=<0 0 0 1>}
 dyn_skirt_l_r_3 {parent=dyn_skirt_l_r_2, loc=<13.6279 0 0>, rot=<0 0 0 1>}
 dyn_skirt_l {parent=pelvis, loc=<1.903 -3.6092 -13.8379>, rot=<0.0682 -0.8468 -0.524 -0.0612>}
 dyn_skirt_l_1 {parent=dyn_skirt_l, loc=<11.8156 0 0>, rot=<0 0 0 1>}
 dyn_skirt_l_2 {parent=dyn_skirt_l_1, loc=<13.3633 0 0>, rot=<0 0 0 1>}
 dyn_skirt_l_3 {parent=dyn_skirt_l_2, loc=<13.6279 0 0>, rot=<0 0 0 1>}
 dyn_skirt_l_l {parent=pelvis, loc=<1.5787 -8.9266 -10.4712>, rot=<0.0749 -0.9291 -0.3597 -0.0418>}
 dyn_skirt_l_l_1 {parent=dyn_skirt_l_l, loc=<11.8156 0 0>, rot=<0 0 0 1>}
 dyn_skirt_l_l_2 {parent=dyn_skirt_l_l_1, loc=<13.3633 0 0>, rot=<0 0 0 1>}
 dyn_skirt_l_l_3 {parent=dyn_skirt_l_l_2, loc=<13.6279 0 0>, rot=<0 0 0 1>}
 dyn_skirt_r_l {parent=pelvis, loc=<1.9848 1.8664 13.7044>, rot=<0.0589 -0.7309 0.6754 0.0788>}
 dyn_skirt_r_l_1 {parent=dyn_skirt_r_l, loc=<11.8156 0 0>, rot=<0 0 0 1>}
 dyn_skirt_r_l_2 {parent=dyn_skirt_r_l_1, loc=<13.3633 0 0>, rot=<0 0 0 -1>}
 dyn_skirt_r_l_3 {parent=dyn_skirt_r_l_2, loc=<13.628 0 0>, rot=<0 0 0 1>}
 dyn_skirt_r {parent=pelvis, loc=<1.903 -3.6092 13.8379>, rot=<0.0682 -0.8466 0.5242 0.0612>}
 dyn_skirt_r_1 {parent=dyn_skirt_r, loc=<11.8156 0 0>, rot=<0 0 0 -1>}
 dyn_skirt_r_2 {parent=dyn_skirt_r_1, loc=<13.3633 0 0>, rot=<0 0 0 1>}
 dyn_skirt_r_3 {parent=dyn_skirt_r_2, loc=<13.628 0 0>, rot=<0 0 0 1>}
 dyn_skirt_r_r {parent=pelvis, loc=<1.5787 -8.9266 10.4712>, rot=<0.0749 -0.9294 0.3589 0.0419>}
 dyn_skirt_r_r_1 {parent=dyn_skirt_r_r, loc=<11.8156 0 0>, rot=<0 0 0 1>}
 dyn_skirt_r_r_2 {parent=dyn_skirt_r_r_1, loc=<13.3634 0 0>, rot=<0 0 0 1>}
 dyn_skirt_r_r_3 {parent=dyn_skirt_r_r_2, loc=<13.6279 0.0001 0>, rot=<0 0 0 1>}
 dyn_skirt_b_b {parent=pelvis, loc=<1.554 -12.88 0.001>, rot=<0.0639 -0.998 0.0002 0>}
 dyn_skirt_b_b_1 {parent=dyn_skirt_b_b, loc=<11.0782 0 0>, rot=<0 0 0 1>}
 dyn_skirt_b_b_2 {parent=dyn_skirt_b_b_1, loc=<12.2213 0 0>, rot=<0 0 0 1>}
 dyn_skirt_b_b_3 {parent=dyn_skirt_b_b_2, loc=<7.9583 0.0001 0>, rot=<0 0 0 1>}
 dyn_belt_sub_l {parent=pelvis, loc=<3.7819 6.6395 -11.3055>, rot=<-0.2336 -0.6674 -0.2096 -0.6753>}
 dyn_belt_sub_r {parent=pelvis, loc=<5.921 7.1407 9.4024>, rot=<0.2715 0.6529 -0.2479 -0.6622>}
]
variables [
 Is_Crouching {type=bool, default=False}
 Is_Crouching_Alpha {type=double, default=0}
 ModifyTransforms {type=FVector, default=(X=0,Y=-2,Z=0)}
 fall {type=double, default=0}
 forwards_backwards_alpha {type=double, default=0}
 is_baseball_sliding_alpha {type=double, default=0}
 is_crouching_not_moving_2d_alpha {type=double, default=0}
 is_in_front_end {type=bool, default=False}
 is_jumping_alpha {type=double, default=0}
 is_knee_sliding_alpha {type=double, default=0}
 is_moving_2d_alpha {type=double, default=0}
 is_rigidbody_active_alpha {type=double, default=0}
 is_skydiving {type=double, default=0}
 is_sprinting_alpha {type=double, default=0}
 left_right_alpha {type=double, default=0}
]
