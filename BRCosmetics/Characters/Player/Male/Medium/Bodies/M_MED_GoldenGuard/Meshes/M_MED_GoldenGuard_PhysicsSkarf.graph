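# Physics asset for the back scarf on the GoldenGuard medium male player cosmetic body, named for simulating its dangling bones.
# Three small capsules form a dyn_scarf_b bone chain anchored to spine_05, alongside two larger capsules covering the upper spine.
# The scarf root has tight swing limits, the lower links swing more widely, and all three barely twist.
# The joint between the two spine bodies is left free in swing and twist.
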PhysicsAsset /BRCosmetics/Characters/Player/Male/Medium/Bodies/M_MED_GoldenGuard/Meshes/M_MED_GoldenGuard_PhysicsSkarf
bodies {
 dyn_scarf_b [{shapeType=Capsule, length=5, radius=5, shapeName=dyn_scarf_b_capsule, center=<5.4087 -5.415 0>}]
 dyn_scarf_b_1 [{shapeType=Capsule, length=5, radius=5, shapeName=dyn_scarf_b_1_capsule, center=<7 -3 0>}]
 dyn_scarf_b_2 [{shapeType=Capsule, length=5, radius=5, shapeName=dyn_scarf_b_2_capsule, center=<7 -3 0>}]
 spine_04 [{shapeType=Capsule, length=15, radius=15, shapeName=spine_04_capsule, center=<1.5191 0.7647 0>}]
 spine_05 [{shapeType=Capsule, length=27.4573, radius=14.8361, center=<1.6872 1.5252 0.1032>, rotation=<rot -0.7505 -90 -0.1767>}]
}
constraints [
 {bone1Name=dyn_scarf_b, bone2Name=spine_05, swing1LimitDegrees=5, swing2LimitDegrees=5, twistLimitDegrees=1}
 {bone1Name=dyn_scarf_b_1, bone2Name=dyn_scarf_b, swing1LimitDegrees=40, swing2LimitDegrees=35, twistLimitDegrees=1}
 {bone1Name=dyn_scarf_b_2, bone2Name=dyn_scarf_b_1, swing1LimitDegrees=40, swing2LimitDegrees=35, twistLimitDegrees=1}
 {
  bone1Name=spine_05
  bone2Name=spine_04
  swing1LimitDegrees=45
  swing1Motion=Free
  swing2LimitDegrees=45
  swing2Motion=Free
  twistLimitDegrees=45
  twistMotion=Free
 }
]
