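# Physics asset for the front skirt panel of the GoldenGuard medium male player body, defining its dynamic bone chain and colliders.
# A sphere root, dyn_f_main, hangs from the pelvis with limited one-axis swing, followed by four capsule segments whose swing loosens and twist tightens toward the tip.
# Pelvis, thigh, calf, forearm and hand shapes supply body volumes, including a large offset pelvis sphere; most limb joints have free angular motion.
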PhysicsAsset /BRCosmetics/Characters/Player/Male/Medium/Bodies/M_MED_GoldenGuard/Meshes/M_MED_GoldenGuard_Physics_FrontSkirt1
bodies {
 calf_l [{shapeType=Capsule, length=30, radius=10, center=<-19.2013 -2.4951 -0.2106>, rotation=<rot 0 -91.0419 -89.2434>}]
 calf_r [{shapeType=Capsule, length=30, radius=10, center=<19.0915 2.1884 0.1601>, rotation=<rot 0 -92.0346 -88.5192>}]
 dyn_f_main [{shapeType=Sphere, radius=4, shapeName=dyn_f_main_sphere, center=<0 -2.9565 0>}]
 dyn_skirt_f_f [{shapeType=Capsule, length=17.75, radius=6, center=<0 -3.5679 0>, rotation=<rot 0 0.0004 0>}]
 dyn_skirt_f_f_1 [{shapeType=Capsule, length=17, radius=6, center=<0 -2.3449 0>}]
 dyn_skirt_f_f_2 [{shapeType=Capsule, length=18, radius=5.25, center=<0 -2.4453 0>}]
 dyn_skirt_f_f_3 [
  {shapeType=Capsule, length=16, radius=4.5, center=<0 -0.9654 0>}
  {shapeType=Capsule, length=16, radius=4.5, shapeName=dyn_skirt_f_f_3_capsule, center=<10.9756 0 0>}
 ]
 hand_l [{shapeType=Capsule, length=12, radius=3, center=<7.9674 0 0>, rotation=<rot 90 0 0>}]
 hand_r [{shapeType=Capsule, length=12, radius=3, center=<-7.9674 0 0>, rotation=<rot 90 0 0>}]
 lowerarm_l [{shapeType=Capsule, length=12, radius=5, center=<7.9674 0 0>, rotation=<rot 90 0 0>}]
 lowerarm_r [{shapeType=Capsule, length=12, radius=5, center=<-7.9674 0 0>, rotation=<rot 90 0 0>}]
 pelvis [
  {shapeType=Sphere, radius=46.6911, shapeName=pelvis_sphere, center=<86.7565 22.891 0>}
  {
   shapeType=Capsule
   length=15.6112
   radius=12.5093
   shapeName=pelvis_capsule
   center=<2.2466 -4.5801 0.1292>
   rotation=<rot 1.2458 -90 1.3916>
  }
  {shapeType=Capsule, length=63.5, radius=15, shapeName=pelvis_capsule, center=<-34.5784 -4.6534 0>, rotation=<rot 90 0 0>}
 ]
 thigh_l [{shapeType=Capsule, length=25.1984, radius=17.25, center=<-21.2167 -10.8108 0.3939>, rotation=<rot 0 92.0686 -89.5718>}]
 thigh_r [{shapeType=Capsule, length=25.1514, radius=17.25, center=<20.8311 10.8108 -0.3911>, rotation=<rot 0 91.0361 -89.9017>}]
}
constraints [
 {bone1Name=calf_l, bone2Name=thigh_l, swing1LimitDegrees=45, swing2LimitDegrees=45, twistLimitDegrees=45}
 {bone1Name=calf_r, bone2Name=thigh_r, swing1LimitDegrees=45, swing2LimitDegrees=45, twistLimitDegrees=45}
 {
  bone1Name=thigh_l
  bone2Name=pelvis
  swing1LimitDegrees=45
  swing1Motion=Free
  swing2LimitDegrees=45
  swing2Motion=Free
  twistLimitDegrees=45
  twistMotion=Free
 }
 {
  bone1Name=thigh_r
  bone2Name=pelvis
  swing1LimitDegrees=45
  swing1Motion=Free
  swing2LimitDegrees=45
  swing2Motion=Free
  twistLimitDegrees=45
  twistMotion=Free
 }
 {
  bone1Name=dyn_f_main
  bone2Name=pelvis
  swing1LimitDegrees=20
  swing2LimitDegrees=45
  swing2Motion=Locked
  twistLimitDegrees=45
  twistMotion=Locked
 }
 {bone1Name=dyn_skirt_f_f, bone2Name=dyn_f_main, swing1LimitDegrees=30, swing2LimitDegrees=20, swing2Motion=Locked, twistLimitDegrees=20}
 {bone1Name=dyn_skirt_f_f_1, bone2Name=dyn_skirt_f_f, swing1LimitDegrees=40, swing2LimitDegrees=20, twistLimitDegrees=20}
 {bone1Name=dyn_skirt_f_f_2, bone2Name=dyn_skirt_f_f_1, swing1LimitDegrees=40, swing2LimitDegrees=20, twistLimitDegrees=5}
 {bone1Name=dyn_skirt_f_f_3, bone2Name=dyn_skirt_f_f_2, swing1LimitDegrees=40, swing2LimitDegrees=20, twistLimitDegrees=5}
 {
  bone1Name=hand_l
  bone2Name=lowerarm_l
  swing1LimitDegrees=45
  swing1Motion=Free
  swing2LimitDegrees=45
  swing2Motion=Free
  twistLimitDegrees=45
  twistMotion=Free
 }
 {
  bone1Name=lowerarm_l
  bone2Name=pelvis
  swing1LimitDegrees=45
  swing1Motion=Free
  swing2LimitDegrees=45
  swing2Motion=Free
  twistLimitDegrees=45
  twistMotion=Free
 }
 {
  bone1Name=lowerarm_r
  bone2Name=pelvis
  swing1LimitDegrees=45
  swing1Motion=Free
  swing2LimitDegrees=45
  swing2Motion=Free
  twistLimitDegrees=45
  twistMotion=Free
 }
 {
  bone1Name=hand_r
  bone2Name=lowerarm_r
  swing1LimitDegrees=45
  swing1Motion=Free
  swing2LimitDegrees=45
  swing2Motion=Free
  twistLimitDegrees=45
  twistMotion=Free
 }
]
