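# Physics asset for the front scarf of the male medium GoldenGuard body cosmetic.
# A chain of four capsule bodies on the dyn_scarf_f bones hangs from spine_04, with narrower swing limit values at the root and minimal twist values throughout.
# Pelvis, spine and thigh capsules approximate the torso and upper legs around the scarf, joined by constraints with all swing and twist locked.
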PhysicsAsset /BRCosmetics/Characters/Player/Male/Medium/Bodies/M_MED_GoldenGuard/Meshes/M_MED_GoldenGuard_Physics_Front_Skarf_R1
bodies {
 dyn_scarf_f [{shapeType=Capsule, length=7.8291, radius=5.5892, shapeName=dyn_scarf_f_capsule, center=<4.7459 0 6.3869>, rotation=<rot 0 0 90>}]
 dyn_scarf_f_1 [{shapeType=Capsule, length=6, radius=6, shapeName=dyn_scarf_f_1_capsule, center=<7.3682 0 4.1609>, rotation=<rot 0 0 90>}]
 dyn_scarf_f_2 [{shapeType=Capsule, length=6, radius=6, shapeName=dyn_scarf_f_2_capsule, center=<7.3682 0 4.1609>, rotation=<rot 0 0 90>}]
 dyn_scarf_f_3 [{shapeType=Capsule, length=6, radius=6, shapeName=dyn_scarf_f_3_capsule, center=<12.4969 0 4.1609>, rotation=<rot 0 0 90>}]
 pelvis [{shapeType=Capsule, length=22.75, radius=11.2297, center=<-1.844 -2.8954 -0.2594>}]
 spine_01 [{shapeType=Capsule, length=53.6925, radius=10.4937, center=<3.96 0.599 0.0013>, rotation=<rot 0 -90 0>}]
 spine_03 [{shapeType=Capsule, length=52.8485, radius=12.6885, center=<3.4512 0.4773 0.0097>, rotation=<rot -0.0615 -90 0.009>}]
 spine_04 [{shapeType=Capsule, length=19.9319, radius=14.8361, center=<4.3068 1.0791 0.1065>, rotation=<rot 0 -90 0>}]
 thigh_l [{shapeType=Capsule, length=32.1984, radius=15, center=<-21.0174 -4.9019 -3.174>, rotation=<rot -0.0011 89.9908 -90.8708>}]
 thigh_r [{shapeType=Capsule, length=35.1984, radius=15, center=<20.8363 6.1169 3.8515>, rotation=<rot 0 90 -87.6396>}]
}
constraints [
 {
  bone1Name=spine_01
  bone2Name=pelvis
  swing1LimitDegrees=45
  swing1Motion=Locked
  swing2LimitDegrees=45
  swing2Motion=Locked
  twistLimitDegrees=45
  twistMotion=Locked
 }
 {
  bone1Name=spine_03
  bone2Name=spine_01
  swing1LimitDegrees=45
  swing1Motion=Locked
  swing2LimitDegrees=45
  swing2Motion=Locked
  twistLimitDegrees=45
  twistMotion=Locked
 }
 {
  bone1Name=spine_04
  bone2Name=spine_03
  swing1LimitDegrees=45
  swing1Motion=Locked
  swing2LimitDegrees=45
  swing2Motion=Locked
  twistLimitDegrees=45
  twistMotion=Locked
 }
 {
  bone1Name=thigh_l
  bone2Name=pelvis
  swing1LimitDegrees=45
  swing1Motion=Locked
  swing2LimitDegrees=45
  swing2Motion=Locked
  twistLimitDegrees=45
  twistMotion=Locked
 }
 {
  bone1Name=thigh_r
  bone2Name=pelvis
  swing1LimitDegrees=45
  swing1Motion=Locked
  swing2LimitDegrees=45
  swing2Motion=Locked
  twistLimitDegrees=45
  twistMotion=Locked
 }
 {bone1Name=dyn_scarf_f, bone2Name=spine_04, swing1LimitDegrees=10, swing2LimitDegrees=20, twistLimitDegrees=1}
 {bone1Name=dyn_scarf_f_2, bone2Name=dyn_scarf_f_1, swing1LimitDegrees=45, swing2LimitDegrees=45, twistLimitDegrees=1}
 {bone1Name=dyn_scarf_f_3, bone2Name=dyn_scarf_f_2, swing1LimitDegrees=45, swing2LimitDegrees=45, twistLimitDegrees=1}
 {bone1Name=dyn_scarf_f_1, bone2Name=dyn_scarf_f, swing1LimitDegrees=45, swing2LimitDegrees=45, twistLimitDegrees=1}
]
