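# Physics asset for the GoldenGuard medium male body, named for a sliding-knee case, that sets up physics for its skirt.
# Capsules cover pelvis, spine_01, thighs and calves, while root carries a large sphere below it plus a capsule.
# Seven four-bone skirt chains (one back, three per side) hang from the pelvis with limited swing and twist.
# Angular-locked cross constraints tie each side's outer chains to its middle chain at every level, so each panel moves together.
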PhysicsAsset /BRCosmetics/Characters/Player/Male/Medium/Bodies/M_MED_GoldenGuard/Meshes/M_MED_GoldenGuard_Physics_SlidingKnee
bodies {
 calf_l [{shapeType=Capsule, length=41.3932, radius=10.6532, center=<-24.0349 -3.3064 -0.1407>, rotation=<rot 0 90 90>}]
 calf_r [{shapeType=Capsule, length=41.3932, radius=10.6532, center=<23.4623 3.3064 0.1407>, rotation=<rot 0 90 -90>}]
 dyn_skirt_b_b [{shapeType=Capsule, length=21.5795, radius=5, shapeName=dyn_skirt_b_b_capsule, center=<7 -3 0>, rotation=<rot 0 -180 180>}]
 dyn_skirt_b_b_1 [{shapeType=Capsule, length=21.5795, radius=5, shapeName=dyn_skirt_b_b_1_capsule, center=<7 -4 0>, rotation=<rot 0 -180 180>}]
 dyn_skirt_b_b_2 [{shapeType=Capsule, length=21.5795, radius=5, shapeName=dyn_skirt_b_b_2_capsule, center=<7 -4 0>, rotation=<rot 0 -180 180>}]
 dyn_skirt_b_b_3 [{shapeType=Capsule, length=21.5795, radius=5, shapeName=dyn_skirt_b_b_3_capsule, center=<13.1428 -4 0>, rotation=<rot 0 -180 180>}]
 dyn_skirt_l [{shapeType=Capsule, length=6, radius=5, shapeName=dyn_skirt_l_capsule, center=<7 -3 0>, rotation=<rot 90 0 0>}]
 dyn_skirt_l_1 [{shapeType=Capsule, length=6, radius=5, shapeName=dyn_skirt_l_1_capsule, center=<7 -4 0>, rotation=<rot 90 0 0>}]
 dyn_skirt_l_2 [{shapeType=Capsule, length=6, radius=5, shapeName=dyn_skirt_l_2_capsule, center=<7 -4 0>, rotation=<rot 90 0 0>}]
 dyn_skirt_l_3 [{shapeType=Capsule, length=6, radius=5, shapeName=dyn_skirt_l_3_capsule, center=<7 -4 0>, rotation=<rot 90 0 0>}]
 dyn_skirt_l_l [{shapeType=Capsule, length=6, radius=5, shapeName=dyn_skirt_l_l_capsule, center=<7 -3 0>, rotation=<rot 90 0 0>}]
 dyn_skirt_l_l_1 [{shapeType=Capsule, length=6, radius=5, shapeName=dyn_skirt_l_l_1_capsule, center=<7 -4 0>, rotation=<rot 90 0 0>}]
 dyn_skirt_l_l_2 [{shapeType=Capsule, length=6, radius=5, shapeName=dyn_skirt_l_l_2_capsule, center=<7 -4 0>, rotation=<rot 90 0 0>}]
 dyn_skirt_l_l_3 [{shapeType=Capsule, length=6, radius=5, shapeName=dyn_skirt_l_3_capsule, center=<7 -4 0>, rotation=<rot 90 0 0>}]
 dyn_skirt_l_r [{shapeType=Capsule, length=6, radius=5, shapeName=dyn_skirt_l_r_capsule, center=<7 -3 0>, rotation=<rot 90 0 0>}]
 dyn_skirt_l_r_1 [{shapeType=Capsule, length=6, radius=5, shapeName=dyn_skirt_l_r_1_capsule, center=<7 -4 0>, rotation=<rot 90 0 0>}]
 dyn_skirt_l_r_2 [{shapeType=Capsule, length=6, radius=5, shapeName=dyn_skirt_l_r_2_capsule, center=<7 -4 0>, rotation=<rot 90 0 0>}]
 dyn_skirt_l_r_3 [{shapeType=Capsule, length=6, radius=5, shapeName=dyn_skirt_l_r_3_capsule, center=<7 -4 0>, rotation=<rot 90 0 0>}]
 dyn_skirt_r [{shapeType=Capsule, length=6, radius=5, shapeName=dyn_skirt_r_capsule, center=<7 -3 0>, rotation=<rot 90 0 0>}]
 dyn_skirt_r_1 [{shapeType=Capsule, length=6, radius=5, shapeName=dyn_skirt_r_1_capsule, center=<7 -4 0>, rotation=<rot 90 11.2629 11.0797>}]
 dyn_skirt_r_2 [{shapeType=Capsule, length=6, radius=5, shapeName=dyn_skirt_r_2_capsule, center=<7 -3 0>, rotation=<rot 90 0 0>}]
 dyn_skirt_r_3 [{shapeType=Capsule, length=6, radius=5, shapeName=dyn_skirt_r_3_capsule, center=<7 -3 0>, rotation=<rot 90 0 0>}]
 dyn_skirt_r_l [{shapeType=Capsule, length=6, radius=5, shapeName=dyn_skirt_r_l_capsule, center=<7 -3 0>, rotation=<rot 90 0 0>}]
 dyn_skirt_r_l_1 [{shapeType=Capsule, length=6, radius=5, shapeName=dyn_skirt_r_l_1_capsule, center=<7 -4 0>, rotation=<rot 90 0 0>}]
 dyn_skirt_r_l_2 [{shapeType=Capsule, length=6, radius=5, shapeName=dyn_skirt_r_l_2_capsule, center=<7 -3 0>, rotation=<rot 90 0 0>}]
 dyn_skirt_r_l_3 [{shapeType=Capsule, length=6, radius=5, shapeName=dyn_skirt_r_l_3_capsule, center=<7 -3 0>, rotation=<rot 90 0 0>}]
 dyn_skirt_r_r [{shapeType=Capsule, length=6, radius=5, shapeName=dyn_skirt_r_r_capsule, center=<7 -3 0>, rotation=<rot 90 0 0>}]
 dyn_skirt_r_r_1 [{shapeType=Capsule, length=6, radius=5, shapeName=dyn_skirt_r_r_1_capsule, center=<7 -4 0>, rotation=<rot 90 0 0>}]
 dyn_skirt_r_r_2 [{shapeType=Capsule, length=6, radius=5, shapeName=dyn_skirt_r_r_2_capsule, center=<7 -3 0>, rotation=<rot 90 0 0>}]
 dyn_skirt_r_r_3 [{shapeType=Capsule, length=6, radius=5, shapeName=dyn_skirt_r_r_3_capsule, center=<7 -3 0>, rotation=<rot 90 0 0>}]
 pelvis [{shapeType=Capsule, length=6.25, radius=11.2297, center=<-1.844 -2.8954 -0.2594>}]
 root [
  {shapeType=Sphere, radius=113.57, shapeName=root_sphere, center=<0 0 -93.6689>}
  {shapeType=Capsule, length=34, radius=14, shapeName=root_capsule, center=<0 0 15.478>}
 ]
 spine_01 [{shapeType=Capsule, length=6.6925, radius=10.4937, center=<3.96 0.599 0.0013>, rotation=<rot 0 -90 0>}]
 thigh_l [
  {shapeType=Capsule, length=27.4484, radius=8.5, center=<-20.9823 -1.8664 -0.8969>, rotation=<rot -0.0011 89.9908 -90.8708>}
  {shapeType=Capsule, length=27.4484, radius=8.5, center=<-20.9723 0.1931 -0.8966>, rotation=<rot -0.0011 89.9908 -90.8708>}
 ]
 thigh_r [
  {shapeType=Capsule, length=27.4484, radius=8.5, center=<20.9821 2.0673 0.3153>, rotation=<rot 0 90 -87.6396>}
  {shapeType=Capsule, length=27.4484, radius=8.5, center=<20.9728 0.0079 0.3214>, rotation=<rot 0 90 -87.6396>}
 ]
}
constraints [
 {bone1Name=spine_01, bone2Name=pelvis, swing1LimitDegrees=45, swing2LimitDegrees=45, twistLimitDegrees=45}
 {bone1Name=thigh_l, bone2Name=pelvis, swing1LimitDegrees=45, swing2LimitDegrees=45, twistLimitDegrees=45}
 {bone1Name=calf_l, bone2Name=thigh_l, swing1LimitDegrees=45, swing2LimitDegrees=45, twistLimitDegrees=45}
 {bone1Name=calf_r, bone2Name=thigh_r, swing1LimitDegrees=45, swing2LimitDegrees=45, twistLimitDegrees=45}
 {bone1Name=thigh_r, bone2Name=pelvis, swing1LimitDegrees=45, swing2LimitDegrees=45, twistLimitDegrees=45}
 {bone1Name=dyn_skirt_b_b, bone2Name=pelvis, swing1LimitDegrees=60, swing2LimitDegrees=10, twistLimitDegrees=20}
 {bone1Name=dyn_skirt_b_b_1, bone2Name=dyn_skirt_b_b, swing1LimitDegrees=30, swing2LimitDegrees=10, twistLimitDegrees=20}
 {bone1Name=dyn_skirt_b_b_2, bone2Name=dyn_skirt_b_b_1, swing1LimitDegrees=60, swing2LimitDegrees=10, twistLimitDegrees=20}
 {bone1Name=dyn_skirt_b_b_3, bone2Name=dyn_skirt_b_b_2, swing1LimitDegrees=60, swing2LimitDegrees=10, twistLimitDegrees=20}
 {bone1Name=dyn_skirt_l, bone2Name=pelvis, swing1LimitDegrees=60, swing2LimitDegrees=10, twistLimitDegrees=20}
 {bone1Name=dyn_skirt_l_1, bone2Name=dyn_skirt_l, swing1LimitDegrees=60, swing2LimitDegrees=10, twistLimitDegrees=20}
 {bone1Name=dyn_skirt_l_2, bone2Name=dyn_skirt_l_1, swing1LimitDegrees=60, swing2LimitDegrees=10, twistLimitDegrees=20}
 {bone1Name=dyn_skirt_l_3, bone2Name=dyn_skirt_l_2, swing1LimitDegrees=60, swing2LimitDegrees=10, twistLimitDegrees=20}
 {bone1Name=dyn_skirt_l_l, bone2Name=pelvis, swing1LimitDegrees=60, swing2LimitDegrees=10, twistLimitDegrees=20}
 {bone1Name=dyn_skirt_l_l_1, bone2Name=dyn_skirt_l_l, swing1LimitDegrees=60, swing2LimitDegrees=10, twistLimitDegrees=20}
 {bone1Name=dyn_skirt_l_l_2, bone2Name=dyn_skirt_l_l_1, swing1LimitDegrees=60, swing2LimitDegrees=10, twistLimitDegrees=20}
 {bone1Name=dyn_skirt_l_l_3, bone2Name=dyn_skirt_l_l_2, swing1LimitDegrees=60, swing2LimitDegrees=10, twistLimitDegrees=20}
 {bone1Name=dyn_skirt_l_r, bone2Name=pelvis, swing1LimitDegrees=20, swing2LimitDegrees=10, twistLimitDegrees=20}
 {bone1Name=dyn_skirt_l_r_1, bone2Name=dyn_skirt_l_r, swing1LimitDegrees=30, swing2LimitDegrees=10, twistLimitDegrees=20}
 {bone1Name=dyn_skirt_l_r_2, bone2Name=dyn_skirt_l_r_1, swing1LimitDegrees=60, swing2LimitDegrees=10, twistLimitDegrees=20}
 {bone1Name=dyn_skirt_l_r_3, bone2Name=dyn_skirt_l_r_2, swing1LimitDegrees=60, swing2LimitDegrees=10, twistLimitDegrees=20}
 {bone1Name=dyn_skirt_r, bone2Name=pelvis, swing1LimitDegrees=60, swing2LimitDegrees=10, twistLimitDegrees=20}
 {bone1Name=dyn_skirt_r_1, bone2Name=dyn_skirt_r, swing1LimitDegrees=60, swing2LimitDegrees=10, twistLimitDegrees=20}
 {bone1Name=dyn_skirt_r_2, bone2Name=dyn_skirt_r_1, swing1LimitDegrees=60, swing2LimitDegrees=10, twistLimitDegrees=20}
 {bone1Name=dyn_skirt_r_3, bone2Name=dyn_skirt_r_2, swing1LimitDegrees=60, swing2LimitDegrees=10, twistLimitDegrees=20}
 {bone1Name=dyn_skirt_r_l, bone2Name=pelvis, swing1LimitDegrees=60, swing2LimitDegrees=10, twistLimitDegrees=20}
 {bone1Name=dyn_skirt_r_l_1, bone2Name=dyn_skirt_r_l, swing1LimitDegrees=60, swing2LimitDegrees=10, twistLimitDegrees=20}
 {bone1Name=dyn_skirt_r_l_2, bone2Name=dyn_skirt_r_l_1, swing1LimitDegrees=60, swing2LimitDegrees=10, twistLimitDegrees=20}
 {bone1Name=dyn_skirt_r_l_3, bone2Name=dyn_skirt_r_l_2, swing1LimitDegrees=60, swing2LimitDegrees=10, twistLimitDegrees=20}
 {bone1Name=dyn_skirt_r_r, bone2Name=pelvis, swing1LimitDegrees=60, swing2LimitDegrees=10, twistLimitDegrees=20}
 {bone1Name=dyn_skirt_r_r_1, bone2Name=dyn_skirt_r_r, swing1LimitDegrees=60, swing2LimitDegrees=10, twistLimitDegrees=20}
 {bone1Name=dyn_skirt_r_r_2, bone2Name=dyn_skirt_r_r_1, swing1LimitDegrees=60, swing2LimitDegrees=10, twistLimitDegrees=20}
 {bone1Name=dyn_skirt_r_r_3, bone2Name=dyn_skirt_r_r_2, swing1LimitDegrees=60, swing2LimitDegrees=10, twistLimitDegrees=20}
 {
  bone1Name=dyn_skirt_l_l
  bone2Name=dyn_skirt_l
  swing1LimitDegrees=45
  swing1Motion=Locked
  swing2LimitDegrees=45
  swing2Motion=Locked
  twistLimitDegrees=45
  twistMotion=Locked
 }
 {
  bone1Name=dyn_skirt_l_r
  bone2Name=dyn_skirt_l
  swing1LimitDegrees=45
  swing1Motion=Locked
  swing2LimitDegrees=45
  swing2Motion=Locked
  twistLimitDegrees=45
  twistMotion=Locked
 }
 {
  bone1Name=dyn_skirt_l_l_1
  bone2Name=dyn_skirt_l_1
  swing1LimitDegrees=45
  swing1Motion=Locked
  swing2LimitDegrees=45
  swing2Motion=Locked
  twistLimitDegrees=45
  twistMotion=Locked
 }
 {
  bone1Name=dyn_skirt_l_r_1
  bone2Name=dyn_skirt_l_1
  swing1LimitDegrees=45
  swing1Motion=Locked
  swing2LimitDegrees=45
  swing2Motion=Locked
  twistLimitDegrees=45
  twistMotion=Locked
 }
 {
  bone1Name=dyn_skirt_l_l_2
  bone2Name=dyn_skirt_l_2
  swing1LimitDegrees=45
  swing1Motion=Locked
  swing2LimitDegrees=45
  swing2Motion=Locked
  twistLimitDegrees=45
  twistMotion=Locked
 }
 {
  bone1Name=dyn_skirt_l_r_2
  bone2Name=dyn_skirt_l_2
  swing1LimitDegrees=45
  swing1Motion=Locked
  swing2LimitDegrees=45
  swing2Motion=Locked
  twistLimitDegrees=45
  twistMotion=Locked
 }
 {
  bone1Name=dyn_skirt_l_l_3
  bone2Name=dyn_skirt_l_3
  swing1LimitDegrees=45
  swing1Motion=Locked
  swing2LimitDegrees=45
  swing2Motion=Locked
  twistLimitDegrees=45
  twistMotion=Locked
 }
 {
  bone1Name=dyn_skirt_l_r_3
  bone2Name=dyn_skirt_l_3
  swing1LimitDegrees=45
  swing1Motion=Locked
  swing2LimitDegrees=45
  swing2Motion=Locked
  twistLimitDegrees=45
  twistMotion=Locked
 }
 {
  bone1Name=dyn_skirt_r_l
  bone2Name=dyn_skirt_r
  swing1LimitDegrees=45
  swing1Motion=Locked
  swing2LimitDegrees=45
  swing2Motion=Locked
  twistLimitDegrees=45
  twistMotion=Locked
 }
 {
  bone1Name=dyn_skirt_r_r
  bone2Name=dyn_skirt_r
  swing1LimitDegrees=45
  swing1Motion=Locked
  swing2LimitDegrees=45
  swing2Motion=Locked
  twistLimitDegrees=45
  twistMotion=Locked
 }
 {
  bone1Name=dyn_skirt_r_l_1
  bone2Name=dyn_skirt_r_1
  swing1LimitDegrees=45
  swing1Motion=Locked
  swing2LimitDegrees=45
  swing2Motion=Locked
  twistLimitDegrees=45
  twistMotion=Locked
 }
 {
  bone1Name=dyn_skirt_r_r_1
  bone2Name=dyn_skirt_r_1
  swing1LimitDegrees=45
  swing1Motion=Locked
  swing2LimitDegrees=45
  swing2Motion=Locked
  twistLimitDegrees=45
  twistMotion=Locked
 }
 {
  bone1Name=dyn_skirt_r_l_2
  bone2Name=dyn_skirt_r_2
  swing1LimitDegrees=45
  swing1Motion=Locked
  swing2LimitDegrees=45
  swing2Motion=Locked
  twistLimitDegrees=45
  twistMotion=Locked
 }
 {
  bone1Name=dyn_skirt_r_r_2
  bone2Name=dyn_skirt_r_2
  swing1LimitDegrees=45
  swing1Motion=Locked
  swing2LimitDegrees=45
  swing2Motion=Locked
  twistLimitDegrees=45
  twistMotion=Locked
 }
 {
  bone1Name=dyn_skirt_r_l_3
  bone2Name=dyn_skirt_r_3
  swing1LimitDegrees=45
  swing1Motion=Locked
  swing2LimitDegrees=45
  swing2Motion=Locked
  twistLimitDegrees=45
  twistMotion=Locked
 }
 {
  bone1Name=dyn_skirt_r_r_3
  bone2Name=dyn_skirt_r_3
  swing1LimitDegrees=45
  swing1Motion=Locked
  swing2LimitDegrees=45
  swing2Motion=Locked
  twistLimitDegrees=45
  twistMotion=Locked
 }
]
