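# Post-process style AnimBP for the GoldenGuard player body, driving its skirt, scarf, bag and shoulder pads.
# It copies the parent mesh pose, then layers bone constraints, RigidBody sims, state-gated bone offsets, bag AnimDynamics and a final Control Rig.
# Baseball sliding switches constraints and physics assets, back bling disables one scarf sim, and LOD 1+ adds a scarf correction.
# The EventGraph updates gravity overrides, state flags and Control Rig alphas.
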AnimBlueprint /BRCosmetics/Characters/Player/Male/Medium/Bodies/M_MED_GoldenGuard/Meshes/M_MED_GoldenGuard_AnimBP
target_skeleton=&/M_MED_GoldenGuard_Skeleton
graphs [
 AnimGraph {nodes [
   n1 = "Transform (Modify) Bone - Bone: dyn_scarf_b"!(
    ComponentPose=Local To Component(LocalPose=Use cached pose 'SP_to Current LOD equal 1'())
    bAlphaBoolEnabled:bool=True
    Translation="0.045,0.066,-0.007"
    Rotation="-0.002,-4.485,-0.005"
    Scale="1,1,1"
   ) -> (Pose) {
    alphaInputType=Bool
    rotationMode=BMM_Replace
    rotationSpace=BCS_BoneSpace
    translationMode=BMM_Replace
    translationSpace=BCS_BoneSpace
    alphaBoolBlend {blendInTime=0.2, blendOption=Linear, blendOutTime=0.3}
    boneToModify {boneName=dyn_scarf_b}
    rotation=<rot -0.002 -4.485 -0.005>
    translation=<0.045 0.066 -0.007>
   }
   n2 = Component To Local(ComponentPose=@n1)
   n3 = Use cached pose 'SP_to Current LOD equal 1'()
   n4 = Get Current LOD equal 1 ?() -> (Current LOD equal 1 ?:bool)
   n5 = "Transform (Modify) Bone - Bone: dyn_scarf_b"(
    ComponentPose=Local To Component(LocalPose=Blend Poses by bool(BlendPose_0=@n2, BlendPose_1=@n3, bActiveValue=@n4) {blendType=Linear})
    bAlphaBoolEnabled=NOT Boolean(A=Get bRigidBodyEnabled())
    Translation="0.045,0.066,-0.007"
    Rotation="-0.002,-4.485,-0.005"
    Scale="1,1,1"
   ) {
    alphaInputType=Bool
    rotationMode=BMM_Replace
    rotationSpace=BCS_BoneSpace
    translationMode=BMM_Replace
    translationSpace=BCS_BoneSpace
    alphaBoolBlend {blendInTime=0.2, blendOption=Linear, blendOutTime=0.3}
    boneToModify {boneName=dyn_scarf_b}
    rotation=<rot -0.002 -4.485 -0.005>
    translation=<0.045 0.066 -0.007>
   }
   n6 = Anim Dynamics - Bone: dyn_bag!(ComponentPose=@n5, Alpha=1, GravityOverride:Vector=Get gravity_override() -> (gravity_override:Vector), ExternalForce="0,0,0") -> (Pose) {
    angularDampingOverride=0.9999
    angularSpringConstant=5000
    bAngularSpring=true
    bGravityOverrideInSimSpace=true
    bOverrideAngularDamping=true
    bUseGravityOverride=true
    simulationSpace=BoneRelative
    alphaBoolBlend {blendOption=Linear}
    boundBone {boneName=dyn_bag}
    componentAppliedLinearAccClamp=<3500 3500 3500>
    componentLinearAccScale=<1.2 1.2 1.2>
    componentLinearVelScale=<1.2 1.2 1.2>
    gravityOverride {z=-980}
    physicsBodyDefinitions [{
      sphereCollisionRadius=0
      boundBone {boneName=dyn_bag}
      boxExtents=<10 20 20>
      constraintSetup {angularLimitsMax=<3 2 2>, angularLimitsMin=<-3 -5 -2>, angularTarget=<1 0 0>}
      localJointOffset=<2 0 0>
     }]
    relativeSpaceBone {boneName=pelvis}
   }
   n7 = "Constraint - dyn_bag (PPPP)"!(
    ComponentPose=@n6
    Alpha=0.2
    ConstraintWeights_0=0.0451
    ConstraintWeights_1:float=0.0118
    ConstraintWeights_2:float=0.6836
    ConstraintWeights_3:float=0.2594
   ) -> (Pose) {
    alphaBoolBlend {blendOption=Linear}
    boneToModify {boneName=dyn_bag}
    constraintSetup [{targetBone {boneName=pelvis}}, {targetBone {boneName=spine_01}}, {targetBone {boneName=dyn_skirt_l_r}}, {targetBone {boneName=dyn_skirt_l}}]
    constraintWeights [1,1,1,1]
   }
   n8 = "Transform (Modify) Bone - Bone: dyn_skirt_f_f"!(ComponentPose=@n7, bAlphaBoolEnabled:bool=NOT Boolean(A=Get bIsInFrontEnd()), Translation="0,-1,0", Rotation="0,0,0", Scale="1,1,1") -> (Pose) {
    alphaInputType=Bool
    rotationMode=BMM_Additive
    rotationSpace=BCS_BoneSpace
    translationMode=BMM_Additive
    translationSpace=BCS_BoneSpace
    alphaBoolBlend {blendInTime=0.2, blendOption=Linear, blendOutTime=0.7}
    boneToModify {boneName=dyn_skirt_f_f}
    translation {y=-1}
   }
   n9 = Get bIsInFrontEnd()
   n10 = Control Rig!(
    Source=Component To Local(ComponentPose=@n8)
    Alpha
    Is_Crouching:bool=Get bIsCrouching()
    Is_Crouching_Alpha:double=Get is_crouching_alpha() -> (is_crouching_alpha:double)
    left_right_alpha:double=Get left_right_alpha() -> (left_right_alpha:double)
    forwards_backwards_alpha:double=Get forwards_backwards_alpha() -> (forwards_backwards_alpha:double)
    is_sprinting_alpha:double=Get is_sprinting_alpha() -> (is_sprinting_alpha:double)
    is_baseball_sliding_alpha:double=Get is_baseball_sliding_alpha() -> (is_baseball_sliding_alpha:float)
    is_knee_sliding_alpha:double=Get is_knee_sliding_alpha() -> (is_knee_sliding_alpha:double)
    is_moving_2d_alpha:double=Get is_moving_2d_alpha() -> (is_moving_2d_alpha:float)
    is_jumping_alpha:double=Get is_jumping_alpha() -> (is_jumping_alpha:double)
    is_skydiving:double=Get is_skydiving_alpha() -> (is_skydiving_alpha:float)
    is_in_front_end:bool=@n9
    fall:double="To Float (Boolean)"(InBool=Get bIsFalling())
    is_rigidbody_active_alpha:double="To Float (Boolean)"(InBool=Get bRigidBodyEnabled())
   ) -> (Pose) {
    alphaBoolBlend {blendOption=Linear}
    controlRigAssetReference {blueprintRigClass=&/M_MED_GoldenGuard_CtrlRig.M_MED_GoldenGuard_CtrlRig_C}
   }
   Output Pose(Result=@n10) {layerGroup=None}
   n11 = "Constraint - dyn_skirt_r_l (RR)"!(
    ComponentPose=Local To Component(LocalPose=Use cached pose 'SP_Is baseball sliding'())
    Alpha=1
    ConstraintWeights_0=1
    ConstraintWeights_1:float=0
   ) -> (Pose) {
    alphaBoolBlend {blendInTime=0.2, blendOption=Linear, blendOutTime=0.2}
    boneToModify {boneName=dyn_skirt_r_l}
    constraintSetup [{transformType=Rotation, targetBone {boneName=thigh_r}}, {transformType=Rotation, targetBone {boneName=pelvis}}]
    constraintWeights [1,1]
   }
   n12 = "Constraint - dyn_skirt_r (RR)"!(ComponentPose=@n11, Alpha=1, ConstraintWeights_0=0.5, ConstraintWeights_1:float=0.5) -> (Pose) {
    alphaBoolBlend {blendInTime=0.2, blendOption=Linear, blendOutTime=0.2}
    boneToModify {boneName=dyn_skirt_r}
    constraintSetup [{transformType=Rotation, targetBone {boneName=dyn_skirt_r_l}}, {transformType=Rotation, targetBone {boneName=dyn_skirt_r_r}}]
    constraintWeights [1,1]
   }
   n13 = "Constraint - dyn_skirt_r_r (RR)"!(ComponentPose=@n12, Alpha=0.5, ConstraintWeights_0=0, ConstraintWeights_1:float=1) -> (Pose) {
    alphaBoolBlend {blendInTime=0.2, blendOption=Linear, blendOutTime=0.2}
    boneToModify {boneName=dyn_skirt_r_r}
    constraintSetup [{transformType=Rotation, targetBone {boneName=thigh_r}}, {transformType=Rotation, targetBone {boneName=pelvis}}]
    constraintWeights [1,1]
   }
   n14 = Component To Local(ComponentPose=@n13)
   n15 = Use cached pose 'SP_Is baseball sliding'()
   n16 = NOT Boolean(A=Get bIsBaseballSliding())
   n17 = "Constraint - dyn_skirt_l_r (RR)"!(
    ComponentPose=Local To Component(LocalPose=Blend Poses by bool(BlendPose_0=@n14, BlendPose_1=@n15, bActiveValue=@n16) {blendType=Linear})
    Alpha=0.5
    ConstraintWeights_0=1
    ConstraintWeights_1:float=0
   ) -> (Pose) {
    alphaBoolBlend {blendOption=Linear}
    boneToModify {boneName=dyn_skirt_l_r}
    constraintSetup [{transformType=Rotation, targetBone {boneName=thigh_l}}, {transformType=Rotation, targetBone {boneName=pelvis}}]
    constraintWeights [1,1]
   }
   n18 = "Constraint - dyn_skirt_l (RR)"!(ComponentPose=@n17, Alpha=1, ConstraintWeights_0=0.5, ConstraintWeights_1:float=0.5) -> (Pose) {
    alphaBoolBlend {blendOption=Linear}
    boneToModify {boneName=dyn_skirt_l}
    constraintSetup [{transformType=Rotation, targetBone {boneName=dyn_skirt_l_r}}, {transformType=Rotation, targetBone {boneName=dyn_skirt_l_l}}]
    constraintWeights [1,1]
   }
   n19 = "Constraint - dyn_skirt_l_l (RR)"!(ComponentPose=@n18, Alpha=0.5, ConstraintWeights_0=0, ConstraintWeights_1:float=1) -> (Pose) {
    alphaBoolBlend {blendOption=Linear}
    boneToModify {boneName=dyn_skirt_l_l}
    constraintSetup [{transformType=Rotation, targetBone {boneName=thigh_l}}, {transformType=Rotation, targetBone {boneName=pelvis}}]
    constraintWeights [1,1]
   }
   n20 = "Constraint - dyn_scarf_main (RR)"!(ComponentPose=@n19, Alpha=0.6, ConstraintWeights_0=0, ConstraintWeights_1:float=0.3) -> (Pose) {
    alphaBoolBlend {blendOption=Linear}
    boneToModify {boneName=dyn_scarf_main}
    constraintSetup [{transformType=Rotation, targetBone {boneName=thigh_l}}, {transformType=Rotation, targetBone {boneName=head}}]
    constraintWeights [1,1]
   }
   n21 = "Constraint - dyn_f_main (PP)"!(ComponentPose=@n20, bAlphaBoolEnabled:bool=True, ConstraintWeights_0=1, ConstraintWeights_1:float=1) -> (Pose) {
    alphaInputType=Bool
    alphaBoolBlend {blendInTime=0.2, blendOption=Linear, blendOutTime=0.2}
    boneToModify {boneName=dyn_f_main}
    constraintSetup [{targetBone {boneName=thigh_r}}, {targetBone {boneName=thigh_l}}]
    constraintWeights [1,1]
   }
   n22 = "Constraint - dyn_scarf_f (PR)"!(ComponentPose=@n21, bAlphaBoolEnabled:bool=True, ConstraintWeights_0=1, ConstraintWeights_1:float=0) -> (Pose) {
    alphaInputType=Bool
    alphaBoolBlend {blendInTime=0.2, blendOption=Linear, blendOutTime=0.2}
    boneToModify {boneName=dyn_scarf_f}
    constraintSetup [{targetBone {boneName=spine_05}}, {transformType=Rotation, targetBone {boneName=thigh_l}}]
    constraintWeights [1,1]
   }
   n23 = "Constraint - dyn_scarf_f_1 (RR)"!(ComponentPose=@n22, bAlphaBoolEnabled:bool=True, ConstraintWeights_0=0, ConstraintWeights_1:float=1) -> (Pose) {
    alphaInputType=Bool
    alphaBoolBlend {blendInTime=0.2, blendOption=Linear, blendOutTime=0.2}
    boneToModify {boneName=dyn_scarf_f_1}
    constraintSetup [{transformType=Rotation, targetBone {boneName=spine_04}}, {transformType=Rotation, targetBone {boneName=dyn_skirt_f_f}}]
    constraintWeights [1,1]
   }
   n24 = RigidBody!(
    ComponentPose=@n23
    bAlphaBoolEnabled:bool
    OverrideWorldGravity:Vector=Get Override World Gravity for Skirt() -> (Override World Gravity for Skirt:Vector)
    ExternalForce
   ) -> (Pose) {
    alphaInputType=Bool
    bClampLinearTranslationLimitToRefPose=true
    bOverrideWorldGravity=true
    overridePhysicsAsset=&/M_MED_GoldenGuard_Physics_FrontSkirt1
    simulationSpace=BaseBoneSpace
    alphaBoolBlend {blendInTime=0.2, blendOption=Linear, blendOutTime=0.2}
    baseBoneRef {boneName=pelvis}
    componentAppliedLinearAccClamp=<500 500 500>
    componentLinearAccScale=<0.1 0.1 0.1>
    componentLinearVelScale=<0.1 0.1 0.1>
    overrideWorldGravity {z=-980}
    simSpaceSettings {
     maxAngularAcceleration=500
     maxAngularVelocity=3
     maxLinearAcceleration=400
     maxLinearVelocity=250
     velocityScaleZ=0
     worldAlpha=1
     externalLinearDragV=<0.1 0.1>
    }
   }
   n25 = Get bIsBaseballSliding()
   n26 = Get gravity_override()
   n27 = RigidBody(ComponentPose=@n24, bAlphaBoolEnabled=NOT Boolean(A=@n25), OverrideWorldGravity=@n26) {
    alphaInputType=Bool
    bClampLinearTranslationLimitToRefPose=true
    bOverrideWorldGravity=true
    overridePhysicsAsset=&/M_MED_GoldenGuard_Physics
    simulationSpace=BaseBoneSpace
    alphaBoolBlend {blendInTime=0.2, blendOption=Linear, blendOutTime=0.2}
    baseBoneRef {boneName=root}
    componentAppliedLinearAccClamp=<3000 3000 3000>
    componentLinearAccScale=<0.45 0.3 0.3>
    componentLinearVelScale=<0.2 0.1 0.1>
    overrideWorldGravity {z=-980}
   }
   n28 = RigidBody(ComponentPose=@n27, bAlphaBoolEnabled=@n25, OverrideWorldGravity=@n26) {
    alphaInputType=Bool
    bClampLinearTranslationLimitToRefPose=true
    bOverrideWorldGravity=true
    overridePhysicsAsset=&/M_MED_GoldenGuard_Physics_SlidingKnee
    simulationSpace=BaseBoneSpace
    alphaBoolBlend {blendInTime=0.2, blendOption=Linear, blendOutTime=0.2}
    baseBoneRef {boneName=pelvis}
    componentAppliedLinearAccClamp=<3000 3000 3000>
    componentLinearAccScale=<0.3 0.3 0.3>
    componentLinearVelScale=<0.1 0.1 0.1>
    overrideWorldGravity {z=-980}
   }
   n29 = RigidBody(ComponentPose=@n28, bAlphaBoolEnabled=NOT Boolean(A=Get IsWearingBackbling() -> (IsWearingBackbling:bool)), OverrideWorldGravity=@n26) {
    alphaInputType=Bool
    bClampLinearTranslationLimitToRefPose=true
    bOverrideWorldGravity=true
    overridePhysicsAsset=&/M_MED_GoldenGuard_PhysicsSkarf
    simulationSpace=BaseBoneSpace
    alphaBoolBlend {blendInTime=0.2, blendOption=Linear, blendOutTime=0.2}
    baseBoneRef {boneName=spine_03}
    componentAppliedLinearAccClamp=<3000 3000 3000>
    componentLinearAccScale=<1 1 1>
    componentLinearVelScale=<0.5 0.5 0.5>
    overrideWorldGravity {z=-980}
   }
   n30 = RigidBody(ComponentPose=@n29, OverrideWorldGravity=@n26) {
    alphaInputType=Bool
    bClampLinearTranslationLimitToRefPose=true
    bOverrideWorldGravity=true
    overridePhysicsAsset=&/M_MED_GoldenGuard_Physics_Front_Skarf_R1
    simulationSpace=BaseBoneSpace
    alphaBoolBlend {blendInTime=0.2, blendOption=Linear, blendOutTime=0.2}
    baseBoneRef {boneName=pelvis}
    componentAppliedLinearAccClamp=<3000 3000 3000>
    componentLinearAccScale=<0.45 0.2 0.2>
    componentLinearVelScale=<0.3 0.2 0.2>
    overrideWorldGravity {z=-1500}
   }
   n31 = "Transform (Modify) Bone - Bone: dyn_bag"!(
    ComponentPose=@n30
    bAlphaBoolEnabled:bool=Get JumpingOrFallingNotSkyDiving() -> (JumpingOrFallingNotSkyDiving:bool)
    Translation="1.726,-0.923,0.311"
    Rotation="6.326,-18.219,-1.164"
    Scale="1,1,1"
   ) -> (Pose) {
    alphaInputType=Bool
    rotationMode=BMM_Additive
    rotationSpace=BCS_BoneSpace
    translationMode=BMM_Additive
    translationSpace=BCS_BoneSpace
    alphaBoolBlend {blendInTime=0.2, blendOption=Linear, blendOutTime=0.2}
    boneToModify {boneName=dyn_bag}
    rotation=<rot 6.326 -18.219 -1.164>
    translation=<1.726 -0.923 0.311>
   }
   n32 = "Transform (Modify) Bone - Bone: dyn_bag"(
    ComponentPose=@n31
    bAlphaBoolEnabled=Get bIsCrouching()
    Translation="-2.034,-8.697,-0.314"
    Rotation="6.635,-46.801,-12.599"
    Scale="1,1,1"
   ) {
    alphaInputType=Bool
    rotationMode=BMM_Additive
    rotationSpace=BCS_BoneSpace
    translationMode=BMM_Additive
    translationSpace=BCS_BoneSpace
    alphaBoolBlend {blendInTime=0.2, blendOption=Linear, blendOutTime=0.2}
    boneToModify {boneName=dyn_bag}
    rotation=<rot 6.635 -46.801 -12.599>
    translation=<-2.034 -8.697 -0.314>
   }
   n33 = "Transform (Modify) Bone - Bone: dyn_bag"(ComponentPose=@n32, bAlphaBoolEnabled=Get bIsMoving2D(), Translation="1.455,-3.147,0.605", Rotation="15.307,-18.768,-0.547", Scale="1,1,1") {
    alphaInputType=Bool
    rotationMode=BMM_Additive
    rotationSpace=BCS_BoneSpace
    translationMode=BMM_Additive
    translationSpace=BCS_BoneSpace
    alphaBoolBlend {blendInTime=0.3, blendOption=Linear, blendOutTime=0.3}
    boneToModify {boneName=dyn_bag}
    rotation=<rot 15.307 -18.768 -0.547>
    translation=<1.455 -3.147 0.605>
   }
   n34 = "Transform (Modify) Bone - Bone: dyn_bag"(
    ComponentPose=@n33
    bAlphaBoolEnabled=Get Is not Moving 2D() -> (Is not Moving 2D:bool)
    Translation="1.622,-0.310,0.143"
    Rotation="7.053,-15.956,-0.690"
    Scale="1,1,1"
   ) {
    alphaInputType=Bool
    rotationMode=BMM_Additive
    rotationSpace=BCS_BoneSpace
    translationMode=BMM_Additive
    translationSpace=BCS_BoneSpace
    alphaBoolBlend {blendInTime=0.1, blendOption=Linear, blendOutTime=0.1}
    boneToModify {boneName=dyn_bag}
    rotation=<rot 7.053 -15.956 -0.69>
    translation=<1.622 -0.31 0.143>
   }
   n35 = "Transform (Modify) Bone - Bone: dyn_bag"(
    ComponentPose=@n34
    bAlphaBoolEnabled=Get Targeting not moving2d() -> (Targeting not moving2d:bool)
    Translation="1.622,-0.310,0.143"
    Rotation="7.053,-15.956,-0.690"
    Scale="1,1,1"
   ) {
    alphaInputType=Bool
    rotationMode=BMM_Additive
    rotationSpace=BCS_BoneSpace
    translationMode=BMM_Additive
    translationSpace=BCS_BoneSpace
    alphaBoolBlend {blendInTime=0.1, blendOption=Linear, blendOutTime=0.1}
    boneToModify {boneName=dyn_bag}
    rotation=<rot 7.053 -15.956 -0.69>
    translation=<1.622 -0.31 0.143>
   }
   n36 = Get bIsTargeting()
   n37 = "Transform (Modify) Bone - Bone: dyn_scarf_b"(ComponentPose=@n35, bAlphaBoolEnabled=@n36, Translation="0.034,-0.642,-4.478", Rotation="0.327,3.416,-16.017", Scale="1,1,1") {
    alphaInputType=Bool
    rotationMode=BMM_Additive
    rotationSpace=BCS_BoneSpace
    translationMode=BMM_Additive
    translationSpace=BCS_BoneSpace
    alphaBoolBlend {blendOption=Linear, blendOutTime=0.4}
    boneToModify {boneName=dyn_scarf_b}
    rotation=<rot 0.327 3.416 -16.017>
    translation=<0.034 -0.642 -4.478>
   }
   n38 = "Transform (Modify) Bone - Bone: dyn_scarf_b_1"!(ComponentPose=@n37, bAlphaBoolEnabled:bool=@n36, Translation="0,0,0", Rotation="0.031,-3.301,-0.000", Scale="1,1,1") -> (Pose) {
    alphaInputType=Bool
    rotationMode=BMM_Additive
    rotationSpace=BCS_BoneSpace
    translationMode=BMM_Additive
    translationSpace=BCS_BoneSpace
    alphaBoolBlend {blendOption=Linear, blendOutTime=0.4}
    boneToModify {boneName=dyn_scarf_b_1}
    rotation=<rot 0.031 -3.301 0>
   }
   n39 = "Transform (Modify) Bone - Bone: dyn_scarf_b"(ComponentPose=@n38, bAlphaBoolEnabled=True, Translation="-0.599,-0.841,0.010", Rotation="-0.000,4.625,-0.000", Scale="1,1,1") {
    alphaInputType=Bool
    rotationMode=BMM_Replace
    rotationSpace=BCS_BoneSpace
    scaleMode=BMM_Additive
    translationMode=BMM_Replace
    translationSpace=BCS_BoneSpace
    alphaBoolBlend {blendInTime=0.5, blendOption=Linear, blendOutTime=0.2}
    boneToModify {boneName=dyn_scarf_b}
    rotation=<rot 0 4.625 0>
    translation=<-0.599 -0.841 0.01>
   }
   n40 = "Transform (Modify) Bone - Bone: dyn_skirt_b_b"!(
    ComponentPose=@n39
    bAlphaBoolEnabled:bool=Get bIsSkyDiving()
    Translation="0.277,-1.529,-0.000"
    Rotation="-0.000,-3.217,0.000"
    Scale="1,1,1"
   ) -> (Pose) {
    alphaInputType=Bool
    rotationMode=BMM_Replace
    rotationSpace=BCS_BoneSpace
    scaleMode=BMM_Additive
    translationMode=BMM_Replace
    translationSpace=BCS_BoneSpace
    alphaBoolBlend {blendInTime=0.5, blendOption=Linear, blendOutTime=0.2}
    boneToModify {boneName=dyn_skirt_b_b}
    rotation=<rot 0 -3.217 0>
    translation=<0.277 -1.529>
   }
   Save cached pose 'SP_main'(Pose=Component To Local(ComponentPose=@n40)) {cacheName=SP_main}
   n41 = "Transform (Modify) Bone - Bone: dyn_shoulder_pad_l"!(
    ComponentPose=Local To Component(LocalPose=Use cached pose 'SP_main'())
    bAlphaBoolEnabled:bool=Get bIsTargeting()
    Translation="0.804,-0.022,0.957"
    Rotation="-3.390,3.454,-0.000"
    Scale="1,1,1"
   ) -> (Pose) {
    alphaInputType=Bool
    rotationMode=BMM_Additive
    rotationSpace=BCS_BoneSpace
    translationMode=BMM_Additive
    translationSpace=BCS_BoneSpace
    alphaBoolBlend {blendInTime=0.2, blendOption=Linear, blendOutTime=0.7}
    boneToModify {boneName=dyn_shoulder_pad_l}
    rotation {pitch=-3.39, yaw=3.454}
    translation=<0.804 -0.022 0.957>
   }
   n42 = "Transform (Modify) Bone - Bone: dyn_shoulder_pad_r"!(ComponentPose=@n41, bAlphaBoolEnabled:bool=Get bIsSkyDiving(), Translation="-1.5,0,0", Rotation="0,0,0", Scale="1,1,1") -> (Pose) {
    alphaInputType=Bool
    rotationMode=BMM_Additive
    rotationSpace=BCS_BoneSpace
    translationMode=BMM_Additive
    translationSpace=BCS_BoneSpace
    alphaBoolBlend {blendInTime=0.2, blendOption=Linear, blendOutTime=0.7}
    boneToModify {boneName=dyn_shoulder_pad_r}
    translation {x=-1.5}
   }
   n43 = "Transform (Modify) Bone - Bone: dyn_shoulder_pad_l"(
    ComponentPose=@n42
    bAlphaBoolEnabled=Get bIsSkydiveFloating()
    Translation="4.723,2.757,-2.081"
    Rotation="-18.439,26.919,-9.269"
    Scale="1,1,1"
   ) {
    alphaInputType=Bool
    rotationMode=BMM_Additive
    rotationSpace=BCS_BoneSpace
    translationMode=BMM_Additive
    translationSpace=BCS_BoneSpace
    alphaBoolBlend {blendInTime=0.2, blendOption=Linear, blendOutTime=0.7}
    boneToModify {boneName=dyn_shoulder_pad_l}
    rotation=<rot -18.439 26.919 -9.269>
    translation=<4.723 2.757 -2.081>
   }
   n44 = Get bIsParachuteOpened()
   n45 = "Transform (Modify) Bone - Bone: dyn_shoulder_pad_l"(ComponentPose=@n43, bAlphaBoolEnabled=@n44, Translation="0.5,-2.757,-2.228", Rotation="0.000,-7.529,0.000", Scale="1,1,1") {
    alphaInputType=Bool
    rotationMode=BMM_Additive
    rotationSpace=BCS_BoneSpace
    translationMode=BMM_Additive
    translationSpace=BCS_BoneSpace
    alphaBoolBlend {blendInTime=0.2, blendOption=Linear, blendOutTime=0.7}
    boneToModify {boneName=dyn_shoulder_pad_l}
    rotation=<rot 0 -7.529 0>
    translation=<0.5 -2.757 -2.228>
   }
   n46 = "Transform (Modify) Bone - Bone: dyn_shoulder_pad_r"(ComponentPose=@n45, bAlphaBoolEnabled=@n44, Translation="-0.5,2.228,2.269", Rotation="0.000,-6.004,-0.000", Scale="1,1,1") {
    alphaInputType=Bool
    rotationMode=BMM_Additive
    rotationSpace=BCS_BoneSpace
    translationMode=BMM_Additive
    translationSpace=BCS_BoneSpace
    alphaBoolBlend {blendInTime=0.2, blendOption=Linear, blendOutTime=0.7}
    boneToModify {boneName=dyn_shoulder_pad_r}
    rotation=<rot 0 -6.004 0>
    translation=<-0.5 2.228 2.269>
   }
   Save cached pose 'SP_to Current LOD equal 1'(Pose=Component To Local(ComponentPose=@n46)) {cacheName=SP_to Current LOD equal 1}
   Save cached pose 'SP_Is baseball sliding'(Pose=Copy Pose From Mesh() {bUseAttachedParent=true}) {cacheName=SP_Is baseball sliding}
   "To Float (Boolean)"(InBool=@n9)
  ]}
 EventGraph {nodes [
   Event BlueprintUpdateAnimation()
   > n1 = Sequence!(execute) -> (
    then_0
    then_1
    then_2:Exec
    then_3:Exec
    then_4:Exec
    then_5:Exec
    then_6:Exec
    then_7:Exec
    then_8:Exec
    then_9:Exec
    then_10:Exec
    then_11:Exec
   )
   n2 = Sequence!(execute=@n1.then_11) -> (then_0, then_1, then_2, then_3)
   Set is_crouching_alpha(execute=@n2.then_2, is_crouching_alpha:double="To Float (Boolean)"(InBool=Get bIsCrouching())) -> (Output_Get:double)
   > Set is_sprinting_alpha(is_sprinting_alpha:double="To Float (Boolean)"(InBool=Get bIsSprinting())) -> (Output_Get:double)
   > Set is_jumping_alpha(is_jumping_alpha:double="To Float (Boolean)"(InBool=Get bIsJumping())) -> (Output_Get:double)
   > Set is_knee_sliding_alpha(is_knee_sliding_alpha:double="To Float (Boolean)"(InBool=Get bIsKneeSliding())) -> (Output_Get:double)
   > Set is_baseball_sliding_alpha(is_baseball_sliding_alpha:float="To Float (Boolean)"(InBool=Get bIsBaseballSliding())) -> (Output_Get:float)
   > Set is_moving_2d_alpha(is_moving_2d_alpha:float="To Float (Boolean)"(InBool=Get bIsMoving2D())) -> (Output_Get:float)
   > Set is_skydiving_alpha(is_skydiving_alpha:float="To Float (Boolean)"(InBool=Get bIsSkyDiving())) -> (Output_Get:float)
   n3 = Gravity Override State Function DEPRECATED(
    execute=@n2.then_0
    CharacterPartAnimInstance=Self-Reference()
    GameTimeInSeconds=GetGameTimeInSeconds()
    GravityOverrideParams=Get Gravity Override Params()
   )
   > Set gravity_override(gravity_override=@n3)
   n4 = PlayerMovement(execute=@n2.then_1, AnimInstance=Self-Reference(), XDotProductCoefficient=1, YDotProductCoefficient=1, ZDotProductCoefficient=1)
   > Set left_right_alpha(left_right_alpha:double=@n4) -> (Output_Get:double)
   > Set forwards_backwards_alpha(forwards_backwards_alpha:double=@n4.YMovement) -> (Output_Get:double)
   n5 = Branch(execute=@n1.then_4, Condition=Get JumpingOrFallingNotSkyDiving())
   > Set Override World Gravity for Skirt(Override World Gravity for Skirt:Vector="0,0,500") -> (Output_Get:Vector)
   n6 = Branch(execute=@n5.else, Condition=Get bIsSkyDiving())
   Set Override World Gravity for Skirt(execute=@n6.else, Override World Gravity for Skirt="0,0,-980")
   n7 = Branch(execute=@n1.then_1, Condition=Get bIsSkyDiving())
   > n8 = Branch(Condition=Get bIsParachuteOpened())
   > Set Override World Gravity for Skirt(Override World Gravity for Skirt="0,0,-50")
   Set Override World Gravity for Skirt(execute=@n8.else, Override World Gravity for Skirt="0,0,500")
   Set Override World Gravity for Skirt(execute=@n7.else, Override World Gravity for Skirt="0,0,-980")
   n9 = Get bIsSkyDiving()
   n10 = Get bIsJumping()
   n11 = Get bIsFalling()
   Set JumpingOrFallingNotSkyDiving(execute=@n1.then_2, JumpingOrFallingNotSkyDiving:bool=AND Boolean(A=OR Boolean(A=@n10, B=@n11), B=NOT Boolean(A=@n9))) -> (Output_Get:bool)
   n12 = Get bIsMoving2D()
   Set Targeting not moving2d(execute=@n1.then_6, Targeting not moving2d:bool=AND Boolean(A=NOT Boolean(A=@n12), B=Get bIsTargeting())) -> (Output_Get:bool)
   n13 = Get CurrentLOD()
   Set Current LOD equal 1 ?(execute=@n1.then_5, Current LOD equal 1 ?:bool="integer >= integer"(A=@n13, B=1)) -> (Output_Get:bool)
   n14 = Get bIsMoving2D()
   n15 = Get bIsTargeting()
   Set Is not Moving 2D(execute=@n1.then_3, Is not Moving 2D:bool=AND Boolean(A=NOT Boolean(A=@n14), B=NOT Boolean(A=@n15))) -> (Output_Get:bool)
   Event OnCharacterCustomizationLoaded()
   n16 = CheckForBackbling()
   > Set IsWearingBackbling(IsWearingBackbling:bool=OR Boolean!(A=@n16, B=@n16.bIsBackblingLong, C:bool=@n16.bIsBackblingACape) -> (ReturnValue)) -> (Output_Get:bool)
   "Not Equal (Integer)"(A=@n13)
  ]}
]
variables [
 Crouching_Right_move {type=boolean}
 Current LOD equal 1 ? {type=boolean}
 forwards_backwards_alpha {type=number}
 Gravity Override Params {
  type=GravityOverrideParamsStruct
  default {
   skydive_FloatingDown {gravityOverride=<0 0 -500>, windAmplitude=<0 0 0>, windOffset=<0 0 0>}
   skydive_GoingUp {gravityOverride=<0 0 -500>, windAmplitude=<0 0 0>}
  }
 }
 gravity_override {type=Vector}
 Is not Moving 2D {type=boolean}
 is_baseball_sliding_alpha {type=number}
 is_crouching_alpha {type=number}
 is_jumping_alpha {type=number}
 is_knee_sliding_alpha {type=number}
 is_moving_2d_alpha {type=number}
 is_skydiving_alpha {type=number}
 is_sprinting_alpha {type=number}
 IsWearingBackbling {type=boolean}
 JumpingOrFallingNotSkyDiving {type=boolean}
 left_right_alpha {type=number}
 Override World Gravity for Skirt {type=Vector}
 Targeting not moving2d {type=boolean}
]
