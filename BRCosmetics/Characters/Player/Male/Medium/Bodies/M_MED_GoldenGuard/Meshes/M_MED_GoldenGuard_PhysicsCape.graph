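# Physics asset for the GoldenGuard medium male body, named for cape physics, providing collision shapes for the character and its dynamic cloth bones.
# Capsules cover spine, neck, pelvis, arms and legs, and a large sphere on root sits offset downward.
# Front, side and back skirt bones carry capsules, with added boxes on the upper back chain.
# Constraints link the bodies in a pelvis-centered hierarchy with all swing and twist motion locked.
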PhysicsAsset /BRCosmetics/Characters/Player/Male/Medium/Bodies/M_MED_GoldenGuard/Meshes/M_MED_GoldenGuard_PhysicsCape
bodies {
 calf_l [{shapeType=Capsule, length=36.75, radius=10.25, shapeName=calf_l_capsule, center=<-25.9259 0 0>, rotation=<rot 90 0 0>}]
 calf_r [{shapeType=Capsule, length=36.75, radius=10.25, shapeName=calf_r_capsule, center=<25.9259 0 0>, rotation=<rot 90 0 0>}]
 dyn_f_main [{shapeType=Capsule, length=20, radius=5, shapeName=dyn_f_main_capsule, center=<0 3.5408 0>, rotation=<rot 90 0 0>}]
 dyn_skirt_b_b [
  {shapeType=Capsule, length=4.5, radius=13.5, shapeName=dyn_skirt_b_b_capsule, center=<0 9.2146 0>}
  {shapeType=Box, extentX=26, extentY=7.5, extentZ=23, shapeName=dyn_skirt_b_b_box, center=<10.3386 2.2034 0>}
 ]
 dyn_skirt_b_b_1 [
  {shapeType=Capsule, length=4.5, radius=13.5, shapeName=dyn_skirt_b_b_1_capsule, center=<0 9.2146 0>}
  {shapeType=Box, extentX=36.5, extentY=5, extentZ=21, shapeName=dyn_skirt_b_b_1_box, center=<18.6095 0.0013 0>}
 ]
 dyn_skirt_b_b_2 [{shapeType=Capsule, length=4.5, radius=13.5, shapeName=dyn_skirt_b_b_2_capsule, center=<0 9.2146 0>}]
 dyn_skirt_b_b_3 [
  {shapeType=Capsule, length=3, radius=13, shapeName=dyn_skirt_b_b_3_capsule, center=<12.499 9.2146 0>}
  {shapeType=Capsule, length=9, radius=7, shapeName=dyn_skirt_b_b_3_capsule, center=<0 2.7564 0>}
 ]
 dyn_skirt_l_l_1 [{shapeType=Capsule, length=4.75, radius=7.75, shapeName=dyn_skirt_l_l_1_capsule, center=<8.3287 1.7928 0>, rotation=<rot 90 0 0>}]
 dyn_skirt_l_l_2 [{shapeType=Capsule, length=4.75, radius=7.75, shapeName=dyn_skirt_l_l_2_capsule, center=<8.3287 1.7928 0>, rotation=<rot 90 0 0>}]
 dyn_skirt_l_l_3 [{shapeType=Capsule, length=4.75, radius=7.75, shapeName=dyn_skirt_l_l_3_capsule, center=<8.3287 1.7928 0>, rotation=<rot 90 0 0>}]
 dyn_skirt_r_r_1 [{shapeType=Capsule, length=4.75, radius=7.75, shapeName=dyn_skirt_r_r_1_capsule, center=<8.3287 1.7928 0>, rotation=<rot 90 0 0>}]
 dyn_skirt_r_r_2 [{shapeType=Capsule, length=4.75, radius=7.75, shapeName=dyn_skirt_r_r_2_capsule, center=<8.3287 1.7928 0>, rotation=<rot 90 0 0>}]
 dyn_skirt_r_r_3 [{shapeType=Capsule, length=4.75, radius=7.75, shapeName=dyn_skirt_r_r_3_capsule, center=<8.3287 1.7928 0>, rotation=<rot 90 0 0>}]
 lowerarm_l [{shapeType=Capsule, length=32.75, radius=8, shapeName=lowerarm_l_capsule, center=<10.6584 0 0>, rotation=<rot 90 0 0>}]
 lowerarm_r [{shapeType=Capsule, length=32.75, radius=8, shapeName=lowerarm_r_capsule, center=<-10.6584 0 0>, rotation=<rot 90 0 0>}]
 neck_02 [{shapeType=Capsule, length=15.1007, radius=10.8784, shapeName=neck_02_capsule, center=<9.2936 0 0>, rotation=<rot 90 0 0>}]
 pelvis [{
   shapeType=Capsule
   length=14.9201
   radius=14.3505
   shapeName=pelvis_capsule
   center=<1.3393 -1.429 0.0135>
   rotation=<rot 5.9116 -90 -1.3062>
  }]
 root [{shapeType=Sphere, radius=108.3846, shapeName=root_sphere, center=<0 0 -74.9588>}]
 spine_01 [{shapeType=Capsule, length=13.9425, radius=10.4937, center=<10.6203 0.2833 0.0016>, rotation=<rot 0.0155 -90 -0.0091>}]
 spine_04 [
  {shapeType=Capsule, length=18.4573, radius=14.8361, center=<10.8708 1.5248 0.1315>, rotation=<rot -0.7505 -90 -0.1767>}
  {shapeType=Capsule, length=15, radius=15, shapeName=spine_04_capsule, center=<1.5191 1.1591 0>}
  {shapeType=Capsule, length=13.25, radius=10.5, shapeName=spine_04_capsule, center=<2.6887 -7.0066 -9.2509>, rotation=<rot -90 0 9.9992>}
 ]
 spine_05 [{shapeType=Capsule, length=27.4573, radius=14.8361, center=<1.6872 1.5252 0.1032>, rotation=<rot -0.7505 -90 -0.1767>}]
 thigh_l [{shapeType=Capsule, length=31.25, radius=10.5, shapeName=thigh_l_capsule, center=<-25.9259 0 0>, rotation=<rot 90 0 0>}]
 thigh_r [{shapeType=Capsule, length=31.25, radius=10.5, shapeName=thigh_r_capsule, center=<25.9259 0 0>, rotation=<rot 90 0 0>}]
 upperarm_l [{shapeType=Capsule, length=14, radius=8, shapeName=upperarm_l_capsule, center=<10.6584 0 0>, rotation=<rot 90 0 0>}]
 upperarm_r [{shapeType=Capsule, length=14, radius=8, shapeName=upperarm_r_capsule, center=<-10.6584 0 0>, rotation=<rot 90 0 0>}]
}
constraints [
 {
  bone1Name=spine_04
  bone2Name=spine_01
  swing1LimitDegrees=45
  swing1Motion=Locked
  swing2LimitDegrees=45
  swing2Motion=Locked
  twistLimitDegrees=45
  twistMotion=Locked
 }
 {
  bone1Name=spine_05
  bone2Name=spine_04
  swing1LimitDegrees=45
  swing1Motion=Locked
  swing2LimitDegrees=45
  swing2Motion=Locked
  twistLimitDegrees=45
  twistMotion=Locked
 }
 {
  bone1Name=neck_02
  bone2Name=spine_05
  swing1LimitDegrees=45
  swing1Motion=Locked
  swing2LimitDegrees=45
  swing2Motion=Locked
  twistLimitDegrees=45
  twistMotion=Locked
 }
 {
  bone1Name=spine_01
  bone2Name=pelvis
  swing1LimitDegrees=45
  swing1Motion=Locked
  swing2LimitDegrees=45
  swing2Motion=Locked
  twistLimitDegrees=45
  twistMotion=Locked
 }
 {
  bone1Name=dyn_f_main
  bone2Name=pelvis
  swing1LimitDegrees=45
  swing1Motion=Locked
  swing2LimitDegrees=45
  swing2Motion=Locked
  twistLimitDegrees=45
  twistMotion=Locked
 }
 {
  bone1Name=pelvis
  bone2Name=root
  swing1LimitDegrees=45
  swing1Motion=Locked
  swing2LimitDegrees=45
  swing2Motion=Locked
  twistLimitDegrees=45
  twistMotion=Locked
 }
 {
  bone1Name=thigh_l
  bone2Name=pelvis
  swing1LimitDegrees=45
  swing1Motion=Locked
  swing2LimitDegrees=45
  swing2Motion=Locked
  twistLimitDegrees=45
  twistMotion=Locked
 }
 {
  bone1Name=calf_l
  bone2Name=thigh_l
  swing1LimitDegrees=45
  swing1Motion=Locked
  swing2LimitDegrees=45
  swing2Motion=Locked
  twistLimitDegrees=45
  twistMotion=Locked
 }
 {
  bone1Name=thigh_r
  bone2Name=pelvis
  swing1LimitDegrees=45
  swing1Motion=Locked
  swing2LimitDegrees=45
  swing2Motion=Locked
  twistLimitDegrees=45
  twistMotion=Locked
 }
 {
  bone1Name=calf_r
  bone2Name=thigh_r
  swing1LimitDegrees=45
  swing1Motion=Locked
  swing2LimitDegrees=45
  swing2Motion=Locked
  twistLimitDegrees=45
  twistMotion=Locked
 }
 {
  bone1Name=dyn_skirt_b_b
  bone2Name=pelvis
  swing1LimitDegrees=45
  swing1Motion=Locked
  swing2LimitDegrees=45
  swing2Motion=Locked
  twistLimitDegrees=45
  twistMotion=Locked
 }
 {
  bone1Name=dyn_skirt_b_b_1
  bone2Name=dyn_skirt_b_b
  swing1LimitDegrees=45
  swing1Motion=Locked
  swing2LimitDegrees=45
  swing2Motion=Locked
  twistLimitDegrees=45
  twistMotion=Locked
 }
 {
  bone1Name=dyn_skirt_b_b_2
  bone2Name=dyn_skirt_b_b_1
  swing1LimitDegrees=45
  swing1Motion=Locked
  swing2LimitDegrees=45
  swing2Motion=Locked
  twistLimitDegrees=45
  twistMotion=Locked
 }
 {
  bone1Name=dyn_skirt_b_b_3
  bone2Name=dyn_skirt_b_b_2
  swing1LimitDegrees=45
  swing1Motion=Locked
  swing2LimitDegrees=45
  swing2Motion=Locked
  twistLimitDegrees=45
  twistMotion=Locked
 }
 {
  bone1Name=upperarm_l
  bone2Name=spine_05
  swing1LimitDegrees=45
  swing1Motion=Locked
  swing2LimitDegrees=45
  swing2Motion=Locked
  twistLimitDegrees=45
  twistMotion=Locked
 }
 {
  bone1Name=lowerarm_l
  bone2Name=upperarm_l
  swing1LimitDegrees=45
  swing1Motion=Locked
  swing2LimitDegrees=45
  swing2Motion=Locked
  twistLimitDegrees=45
  twistMotion=Locked
 }
 {
  bone1Name=upperarm_r
  bone2Name=spine_05
  swing1LimitDegrees=45
  swing1Motion=Locked
  swing2LimitDegrees=45
  swing2Motion=Locked
  twistLimitDegrees=45
  twistMotion=Locked
 }
 {
  bone1Name=lowerarm_r
  bone2Name=upperarm_r
  swing1LimitDegrees=45
  swing1Motion=Locked
  swing2LimitDegrees=45
  swing2Motion=Locked
  twistLimitDegrees=45
  twistMotion=Locked
 }
 {
  bone1Name=dyn_skirt_l_l_2
  bone2Name=dyn_skirt_l_l_1
  swing1LimitDegrees=45
  swing1Motion=Locked
  swing2LimitDegrees=45
  swing2Motion=Locked
  twistLimitDegrees=45
  twistMotion=Locked
 }
 {
  bone1Name=dyn_skirt_l_l_3
  bone2Name=dyn_skirt_l_l_2
  swing1LimitDegrees=45
  swing1Motion=Locked
  swing2LimitDegrees=45
  swing2Motion=Locked
  twistLimitDegrees=45
  twistMotion=Locked
 }
 {
  bone1Name=dyn_skirt_l_l_1
  bone2Name=pelvis
  swing1LimitDegrees=45
  swing1Motion=Locked
  swing2LimitDegrees=45
  swing2Motion=Locked
  twistLimitDegrees=45
  twistMotion=Locked
 }
 {
  bone1Name=dyn_skirt_r_r_1
  bone2Name=pelvis
  swing1LimitDegrees=45
  swing1Motion=Locked
  swing2LimitDegrees=45
  swing2Motion=Locked
  twistLimitDegrees=45
  twistMotion=Locked
 }
 {
  bone1Name=dyn_skirt_r_r_2
  bone2Name=dyn_skirt_r_r_1
  swing1LimitDegrees=45
  swing1Motion=Locked
  swing2LimitDegrees=45
  swing2Motion=Locked
  twistLimitDegrees=45
  twistMotion=Locked
 }
 {
  bone1Name=dyn_skirt_r_r_3
  bone2Name=dyn_skirt_r_r_2
  swing1LimitDegrees=45
  swing1Motion=Locked
  swing2LimitDegrees=45
  swing2Motion=Locked
  twistLimitDegrees=45
  twistMotion=Locked
 }
]
